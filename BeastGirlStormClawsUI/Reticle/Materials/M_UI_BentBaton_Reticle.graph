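# UI material that draws the mirrored Bent Baton reticle for BeastGirlStormClaws, including three ribbon layers with their own UV subgraphs.
# 9-slice UVs feed symmetry functions with parameterized offsets, and MSDF textures become distance fields antialiased at adjustable thresholds.
# Scalar parameters blend between shapes, outlines and ribbon rotations.
# Radial gradients, circle SDFs and floor-based steps build additional masks, while lerped vector parameters supply color.
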Material /BeastGirlStormClawsUI/Reticle/Materials/M_UI_BentBaton_Reticle
graph {
 expressions {
  MaterialExpressionAdd_0 {
   class=MaterialExpressionAdd
   inputs [
    A {from=MaterialExpressionScalarParameter_2}
    B {from=""}
   ]
   properties {constB=-0.05}
  }
  MaterialExpressionAdd_1 {
   class=MaterialExpressionAdd
   inputs [
    A {from=MaterialExpressionScalarParameter_3}
    B {from=""}
   ]
   properties {constB=-0.1}
  }
  MaterialExpressionAdd_10 {
   class=MaterialExpressionAdd
   inputs [
    A {from=MaterialExpressionConvert_5}
    B {from=MaterialExpressionMultiply_10}
   ]
   properties {constB=0}
  }
  MaterialExpressionAdd_11 {
   class=MaterialExpressionAdd
   inputs [
    A {from=MaterialExpressionScalarParameter_19}
    B {from=""}
   ]
   properties {constB=-0.05}
  }
  MaterialExpressionAdd_12 {
   class=MaterialExpressionAdd
   inputs [
    A {from=MaterialExpressionScalarParameter_21}
    B {from=""}
   ]
   properties {constB=-0.05}
  }
  MaterialExpressionAdd_13 {
   class=MaterialExpressionAdd
   inputs [
    A {from=MaterialExpressionScalarParameter_23}
    B {from=""}
   ]
   properties {constB=-0.05}
  }
  MaterialExpressionAdd_14 {
   class=MaterialExpressionAdd
   inputs [
    A {from=MaterialExpressionConstant_5}
    B {from=MaterialExpressionNamedRerouteUsage_18}
   ]
  }
  MaterialExpressionAdd_16 {
   class=MaterialExpressionAdd
   inputs [
    A {from=MaterialExpressionReroute_46}
    B {from=MaterialExpressionFrac_0}
   ]
  }
  MaterialExpressionAdd_17 {
   class=MaterialExpressionAdd
   inputs [
    A {from=MaterialExpressionMultiply_14}
    B {from=MaterialExpressionReroute_12}
   ]
  }
  MaterialExpressionAdd_18 {
   class=MaterialExpressionAdd
   inputs [
    A {from=MaterialExpressionMultiply_16}
    B {from=MaterialExpressionMultiply_15}
   ]
  }
  MaterialExpressionAdd_19 {
   class=MaterialExpressionAdd
   inputs [
    A {from=MaterialExpressionSaturate_0}
    B {from=MaterialExpressionReroute_15}
   ]
  }
  MaterialExpressionAdd_2 {
   class=MaterialExpressionAdd
   inputs [
    A {from=MaterialExpressionConvert_0}
    B {from=""}
   ]
   properties {constB=0}
  }
  MaterialExpressionAdd_20 {
   class=MaterialExpressionAdd
   inputs [
    A {from=MaterialExpressionMultiply_20}
    B {from=MaterialExpressionMultiply_38}
   ]
  }
  MaterialExpressionAdd_21 {
   class=MaterialExpressionAdd
   inputs [
    A {from=MaterialExpressionMultiply_22}
    B {from=MaterialExpressionReroute_19}
   ]
  }
  MaterialExpressionAdd_22 {
   class=MaterialExpressionAdd
   inputs [
    A {from=MaterialExpressionSaturate_1}
    B {from=MaterialExpressionReroute_21}
   ]
  }
  MaterialExpressionAdd_23 {
   class=MaterialExpressionAdd
   inputs [
    A {from=MaterialExpressionMultiply_26}
    B {from=MaterialExpressionMultiply_40}
   ]
  }
  MaterialExpressionAdd_24 {
   class=MaterialExpressionAdd
   inputs [
    A {from=MaterialExpressionMultiply_28}
    B {from=MaterialExpressionReroute_24}
   ]
  }
  MaterialExpressionAdd_25 {
   class=MaterialExpressionAdd
   inputs [
    A {from=MaterialExpressionSaturate_2}
    B {from=MaterialExpressionReroute_27}
   ]
  }
  MaterialExpressionAdd_26 {
   class=MaterialExpressionAdd
   inputs [
    A {from=MaterialExpressionMultiply_32}
    B {from=MaterialExpressionMultiply_41}
   ]
  }
  MaterialExpressionAdd_27 {
   class=MaterialExpressionAdd
   inputs [
    A {from=MaterialExpressionMultiply_34}
    B {from=MaterialExpressionReroute_32}
   ]
  }
  MaterialExpressionAdd_28 {
   class=MaterialExpressionAdd
   inputs [
    A {from=MaterialExpressionMaterialFunctionCall_31, output=Result}
    B {from=MaterialExpressionReroute_9}
   ]
  }
  MaterialExpressionAdd_29 {
   class=MaterialExpressionAdd
   inputs [
    A {from=MaterialExpressionOneMinus_0}
    B {from=MaterialExpressionAppendVector_1}
   ]
  }
  MaterialExpressionAdd_3 {
   class=MaterialExpressionAdd
   inputs [
    A {from=MaterialExpressionConvert_2}
    B {from=MaterialExpressionMultiply_65}
   ]
  }
  MaterialExpressionAdd_32 {
   class=MaterialExpressionAdd
   inputs [
    A {from=MaterialExpressionFrac_1}
    B {from=MaterialExpressionReroute_56}
   ]
  }
  MaterialExpressionAdd_33 {
   class=MaterialExpressionAdd
   inputs [
    A {from=MaterialExpressionMultiply_0}
    B {from=MaterialExpressionAdd_4}
   ]
  }
  MaterialExpressionAdd_34 {
   class=MaterialExpressionAdd
   inputs [
    A {from=MaterialExpressionOneMinus_2}
    B {from=MaterialExpressionAppendVector_3}
   ]
  }
  MaterialExpressionAdd_35 {
   class=MaterialExpressionAdd
   inputs [
    A {from=MaterialExpressionFrac_2}
    B {from=MaterialExpressionReroute_66}
   ]
  }
  MaterialExpressionAdd_36 {
   class=MaterialExpressionAdd
   inputs [
    A {from=MaterialExpressionOneMinus_4}
    B {from=MaterialExpressionAppendVector_5}
   ]
  }
  MaterialExpressionAdd_4 {
   class=MaterialExpressionAdd
   inputs [
    A {from=MaterialExpressionMultiply_42}
    B {from=MaterialExpressionLinearInterpolate_9}
   ]
  }
  MaterialExpressionAdd_7 {
   class=MaterialExpressionAdd
   inputs [
    A {from=MaterialExpressionConvert_1}
    B {from=MaterialExpressionMultiply_62}
   ]
   properties {constB=0}
  }
  MaterialExpressionAdd_8 {
   class=MaterialExpressionAdd
   inputs [
    A {from=MaterialExpressionConvert_3}
    B {from=MaterialExpressionMultiply_45}
   ]
   properties {constB=0}
  }
  MaterialExpressionAdd_9 {
   class=MaterialExpressionAdd
   inputs [
    A {from=MaterialExpressionConvert_4}
    B {from=MaterialExpressionMultiply_8}
   ]
   properties {constB=0}
  }
  MaterialExpressionAppendVector_0 {
   class=MaterialExpressionAppendVector
   inputs [
    A {from=MaterialExpressionMaterialFunctionCall_70, output=R}
    B {from=MaterialExpressionOneMinus_1, output=""}
   ]
  }
  MaterialExpressionAppendVector_1 {
   class=MaterialExpressionAppendVector
   inputs [
    A {from=MaterialExpressionMaterialFunctionCall_71, output=R}
    B {from=MaterialExpressionMultiply_35, output=""}
   ]
  }
  MaterialExpressionAppendVector_2 {
   class=MaterialExpressionAppendVector
   inputs [
    A {from=MaterialExpressionMaterialFunctionCall_89, output=R}
    B {from=MaterialExpressionOneMinus_3, output=""}
   ]
  }
  MaterialExpressionAppendVector_3 {
   class=MaterialExpressionAppendVector
   inputs [
    A {from=MaterialExpressionMaterialFunctionCall_90, output=R}
    B {from=MaterialExpressionMultiply_39, output=""}
   ]
  }
  MaterialExpressionAppendVector_4 {
   class=MaterialExpressionAppendVector
   inputs [
    A {from=MaterialExpressionMaterialFunctionCall_93, output=R}
    B {from=MaterialExpressionOneMinus_5, output=""}
   ]
  }
  MaterialExpressionAppendVector_5 {
   class=MaterialExpressionAppendVector
   inputs [
    A {from=MaterialExpressionMaterialFunctionCall_94, output=R}
    B {from=MaterialExpressionMultiply_50, output=""}
   ]
  }
  MaterialExpressionAppendVector_6 {
   class=MaterialExpressionAppendVector
   inputs [
    A {from=MaterialExpressionMultiply_61, output=""}
    B {from=MaterialExpressionScalarParameter_44, output=""}
   ]
  }
  MaterialExpressionClamp_0 {
   class=MaterialExpressionClamp
   inputs [
    None {from=MaterialExpressionMaterialFunctionCall_28, output=UVs}
    Min {from="", output=""}
    Max {from="", output=""}
   ]
   properties {maxDefault=0.99, minDefault=0.01}
  }
  MaterialExpressionClamp_1 {
   class=MaterialExpressionClamp
   inputs [
    None {from=MaterialExpressionMaterialFunctionCall_29, output=UVs}
    Min {from="", output=""}
    Max {from="", output=""}
   ]
   properties {maxDefault=0.99, minDefault=0.01}
  }
  MaterialExpressionClamp_2 {
   class=MaterialExpressionClamp
   inputs [
    None {from=MaterialExpressionMaterialFunctionCall_30, output=UVs}
    Min {from="", output=""}
    Max {from="", output=""}
   ]
   properties {maxDefault=0.99, minDefault=0.01}
  }
  MaterialExpressionClamp_3 {
   class=MaterialExpressionClamp
   inputs [
    None {from=MaterialExpressionMaterialFunctionCall_27, output=UVs}
    Min {from="", output=""}
    Max {from="", output=""}
   ]
   properties {maxDefault=0.99, minDefault=0.01}
  }
  MaterialExpressionClamp_4 {
   class=MaterialExpressionClamp
   inputs [
    None {from=MaterialExpressionMaterialFunctionCall_86, output=UVs}
    Min {from="", output=""}
    Max {from="", output=""}
   ]
   properties {maxDefault=0.99, minDefault=0.01}
  }
  MaterialExpressionClamp_5 {
   class=MaterialExpressionClamp
   inputs [
    None {from=MaterialExpressionMaterialFunctionCall_91, output=UVs}
    Min {from="", output=""}
    Max {from="", output=""}
   ]
   properties {maxDefault=0.99, minDefault=0.01}
  }
  MaterialExpressionClamp_6 {
   class=MaterialExpressionClamp
   inputs [
    None {from=MaterialExpressionMaterialFunctionCall_92, output=UVs}
    Min {from="", output=""}
    Max {from="", output=""}
   ]
   properties {maxDefault=0.99, minDefault=0.01}
  }
  MaterialExpressionComponentMask_0 {class=MaterialExpressionComponentMask, inputs [None {from=MaterialExpressionReroute_36}], properties {r=true}}
  MaterialExpressionComponentMask_1 {class=MaterialExpressionComponentMask, inputs [None {from=MaterialExpressionReroute_51}], properties {r=true}}
  MaterialExpressionComponentMask_2 {class=MaterialExpressionComponentMask, inputs [None {from=MaterialExpressionReroute_65}], properties {r=true}}
  MaterialExpressionComposite_0 {
   class=MaterialExpressionComposite
   inputs [
    Offset {from=MaterialExpressionAdd_8, output=""}
    Scale {from=MaterialExpressionNamedRerouteUsage_15, output=""}
    Rotation {from=MaterialExpressionLinearInterpolate_11, output=""}
    Pivot {from=MaterialExpressionConstant2Vector_32, output=""}
   ]
   properties {subgraphName=FirstRibbonUvs}
  }
  MaterialExpressionComposite_1 {
   class=MaterialExpressionComposite
   inputs [
    Offset {from=MaterialExpressionAdd_9, output=""}
    Scale {from=MaterialExpressionNamedRerouteUsage_17, output=""}
    Rotation {from=MaterialExpressionLinearInterpolate_13, output=""}
    Pivot {from=MaterialExpressionConstant2Vector_36, output=""}
   ]
   properties {subgraphName=SecondRibbonUVs}
  }
  MaterialExpressionComposite_2 {
   class=MaterialExpressionComposite
   inputs [
    Offset {from=MaterialExpressionReroute_5, output=""}
    Scale {from=MaterialExpressionNamedRerouteUsage_20, output=""}
    Rotation {from=MaterialExpressionLinearInterpolate_23, output=""}
    Pivot {from=MaterialExpressionConstant2Vector_39, output=""}
   ]
   properties {subgraphName=ThirdRibbonUvs}
  }
  MaterialExpressionConstant2Vector_10 {class=MaterialExpressionConstant2Vector, properties {g=1, r=2}}
  MaterialExpressionConstant2Vector_11 {class=MaterialExpressionConstant2Vector, properties {g=1, r=2}}
  MaterialExpressionConstant2Vector_12 {class=MaterialExpressionConstant2Vector, properties {g=1}}
  MaterialExpressionConstant2Vector_13 {class=MaterialExpressionConstant2Vector, properties {g=0.395, r=0.082}}
  MaterialExpressionConstant2Vector_14 {class=MaterialExpressionConstant2Vector, properties {g=0.5, r=1}}
  MaterialExpressionConstant2Vector_15 {class=MaterialExpressionConstant2Vector, properties {g=-0.05, r=-0.75}}
  MaterialExpressionConstant2Vector_16 {class=MaterialExpressionConstant2Vector, properties {g=0.33}}
  MaterialExpressionConstant2Vector_17 {class=MaterialExpressionConstant2Vector, properties {g=0.374, r=0.106}}
  MaterialExpressionConstant2Vector_18 {class=MaterialExpressionConstant2Vector, properties {g=0.5, r=1}}
  MaterialExpressionConstant2Vector_19 {class=MaterialExpressionConstant2Vector, properties {g=-0.05, r=-0.75}}
  MaterialExpressionConstant2Vector_2 {class=MaterialExpressionConstant2Vector, properties {r=0.15}}
  MaterialExpressionConstant2Vector_20 {class=MaterialExpressionConstant2Vector, properties {g=0.66, r=0.33}}
  MaterialExpressionConstant2Vector_21 {class=MaterialExpressionConstant2Vector, properties {g=0.343, r=0.172}}
  MaterialExpressionConstant2Vector_22 {class=MaterialExpressionConstant2Vector, properties {g=0.5, r=1}}
  MaterialExpressionConstant2Vector_23 {class=MaterialExpressionConstant2Vector, properties {g=-0.05, r=-0.75}}
  MaterialExpressionConstant2Vector_24 {class=MaterialExpressionConstant2Vector, properties {g=1, r=0.66}}
  MaterialExpressionConstant2Vector_25 {class=MaterialExpressionConstant2Vector, properties {g=0.343, r=0.215}}
  MaterialExpressionConstant2Vector_26 {class=MaterialExpressionConstant2Vector, properties {g=0.5, r=1}}
  MaterialExpressionConstant2Vector_27 {class=MaterialExpressionConstant2Vector, properties {g=-0.05, r=-0.75}}
  MaterialExpressionConstant2Vector_28 {class=MaterialExpressionConstant2Vector, properties {g=1.5, r=0.5}}
  MaterialExpressionConstant2Vector_29 {class=MaterialExpressionConstant2Vector, properties {g=1.4, r=2.5}}
  MaterialExpressionConstant2Vector_3 {class=MaterialExpressionConstant2Vector, properties {g=1, r=2}}
  MaterialExpressionConstant2Vector_30 {class=MaterialExpressionConstant2Vector, properties {g=0.05}}
  MaterialExpressionConstant2Vector_31 {class=MaterialExpressionConstant2Vector, properties {g=1, r=2}}
  MaterialExpressionConstant2Vector_32 {class=MaterialExpressionConstant2Vector, properties {g=0.25, r=0.55}}
  MaterialExpressionConstant2Vector_34 {class=MaterialExpressionConstant2Vector, properties {g=1, r=2}}
  MaterialExpressionConstant2Vector_35 {class=MaterialExpressionConstant2Vector, properties {g=1, r=2}}
  MaterialExpressionConstant2Vector_36 {class=MaterialExpressionConstant2Vector, properties {g=0.25, r=0.6}}
  MaterialExpressionConstant2Vector_37 {class=MaterialExpressionConstant2Vector, properties {g=1, r=2}}
  MaterialExpressionConstant2Vector_38 {class=MaterialExpressionConstant2Vector, properties {g=1, r=2}}
  MaterialExpressionConstant2Vector_39 {class=MaterialExpressionConstant2Vector, properties {g=0.3, r=0.63}}
  MaterialExpressionConstant2Vector_5 {class=MaterialExpressionConstant2Vector, properties {g=0.03, r=0.05}}
  MaterialExpressionConstant2Vector_6 {class=MaterialExpressionConstant2Vector, properties {r=0.2}}
  MaterialExpressionConstant2Vector_7 {class=MaterialExpressionConstant2Vector, properties {g=0.05, r=0.36}}
  MaterialExpressionConstant2Vector_8 {class=MaterialExpressionConstant2Vector, properties {g=0.15, r=0.49}}
  MaterialExpressionConstant2Vector_9 {class=MaterialExpressionConstant2Vector, properties {g=1, r=2}}
  MaterialExpressionConstantBiasScale_1 {class=MaterialExpressionConstantBiasScale, inputs [None {from=MaterialExpressionSine_0}]}
  MaterialExpressionConstant_1 {class=MaterialExpressionConstant, properties {r=0.4675}}
  MaterialExpressionConstant_13 {class=MaterialExpressionConstant}
  MaterialExpressionConstant_14 {class=MaterialExpressionConstant, properties {r=1.1}}
  MaterialExpressionConstant_15 {class=MaterialExpressionConstant, properties {r=0.33}}
  MaterialExpressionConstant_16 {class=MaterialExpressionConstant, properties {r=1.1}}
  MaterialExpressionConstant_17 {class=MaterialExpressionConstant, properties {r=0.66}}
  MaterialExpressionConstant_18 {class=MaterialExpressionConstant, properties {r=1.1}}
  MaterialExpressionConstant_19 {class=MaterialExpressionConstant, properties {r=0.1}}
  MaterialExpressionConstant_2 {class=MaterialExpressionConstant, properties {r=3}}
  MaterialExpressionConstant_21 {class=MaterialExpressionConstant, properties {r=0.4675}}
  MaterialExpressionConstant_3 {class=MaterialExpressionConstant, properties {r=3}}
  MaterialExpressionConstant_4 {class=MaterialExpressionConstant, properties {r=3}}
  MaterialExpressionConstant_5 {class=MaterialExpressionConstant, properties {r=0.45}}
  MaterialExpressionConstant_6 {class=MaterialExpressionConstant, properties {r=0.4675}}
  MaterialExpressionConstant_7 {class=MaterialExpressionConstant, properties {r=3}}
  MaterialExpressionConstant_8 {class=MaterialExpressionConstant, properties {r=0.4675}}
  MaterialExpressionConstant_9 {class=MaterialExpressionConstant, properties {r=3}}
  MaterialExpressionConvert_0 {
   class=MaterialExpressionConvert
   inputs [
    None {from=MaterialExpressionScalarParameter_27}
    None {from=MaterialExpressionScalarParameter_90}
   ]
   properties {
    convertInputs [
     {
      type=Scalar
      expressionInput {
       expression=&self:MaterialExpressionScalarParameter_27
       inputName=None
       mask=0
       maskA=0
       maskB=0
       maskG=0
       maskR=0
       outputIndex=0
      }
     }
     {
      type=Scalar
      expressionInput {
       expression=&self:MaterialExpressionScalarParameter_90
       inputName=None
       mask=0
       maskA=0
       maskB=0
       maskG=0
       maskR=0
       outputIndex=0
      }
     }
    ]
    convertOutputs [{type=Vector2}]
   }
  }
  MaterialExpressionConvert_1 {
   class=MaterialExpressionConvert
   inputs [
    None {from=MaterialExpressionScalarParameter_11}
    None {from=MaterialExpressionScalarParameter_91}
   ]
   properties {
    convertInputs [
     {
      type=Scalar
      expressionInput {
       expression=&self:MaterialExpressionScalarParameter_11
       inputName=None
       mask=0
       maskA=0
       maskB=0
       maskG=0
       maskR=0
       outputIndex=0
      }
     }
     {
      type=Scalar
      expressionInput {
       expression=&self:MaterialExpressionScalarParameter_91
       inputName=None
       mask=0
       maskA=0
       maskB=0
       maskG=0
       maskR=0
       outputIndex=0
      }
     }
    ]
    convertOutputs [{type=Vector2}]
   }
  }
  MaterialExpressionConvert_2 {
   class=MaterialExpressionConvert
   inputs [
    None {from=MaterialExpressionScalarParameter_12}
    None {from=MaterialExpressionScalarParameter_89}
   ]
   properties {
    convertInputs [
     {
      type=Scalar
      expressionInput {
       expression=&self:MaterialExpressionScalarParameter_12
       inputName=None
       mask=0
       maskA=0
       maskB=0
       maskG=0
       maskR=0
       outputIndex=0
      }
     }
     {
      type=Scalar
      expressionInput {
       expression=&self:MaterialExpressionScalarParameter_89
       inputName=None
       mask=0
       maskA=0
       maskB=0
       maskG=0
       maskR=0
       outputIndex=0
      }
     }
    ]
    convertOutputs [{type=Vector2}]
   }
  }
  MaterialExpressionConvert_3 {
   class=MaterialExpressionConvert
   inputs [
    None {from=MaterialExpressionScalarParameter_13}
    None {from=MaterialExpressionScalarParameter_93}
   ]
   properties {
    convertInputs [
     {
      type=Scalar
      expressionInput {
       expression=&self:MaterialExpressionScalarParameter_13
       inputName=None
       mask=0
       maskA=0
       maskB=0
       maskG=0
       maskR=0
       outputIndex=0
      }
     }
     {
      type=Scalar
      expressionInput {
       expression=&self:MaterialExpressionScalarParameter_93
       inputName=None
       mask=0
       maskA=0
       maskB=0
       maskG=0
       maskR=0
       outputIndex=0
      }
     }
    ]
    convertOutputs [{type=Vector2}]
   }
  }
  MaterialExpressionConvert_4 {
   class=MaterialExpressionConvert
   inputs [
    None {from=MaterialExpressionScalarParameter_14}
    None {from=MaterialExpressionScalarParameter_94}
   ]
   properties {
    convertInputs [
     {
      type=Scalar
      expressionInput {
       expression=&self:MaterialExpressionScalarParameter_14
       inputName=None
       mask=0
       maskA=0
       maskB=0
       maskG=0
       maskR=0
       outputIndex=0
      }
     }
     {
      type=Scalar
      expressionInput {
       expression=&self:MaterialExpressionScalarParameter_94
       inputName=None
       mask=0
       maskA=0
       maskB=0
       maskG=0
       maskR=0
       outputIndex=0
      }
     }
    ]
    convertOutputs [{type=Vector2}]
   }
  }
  MaterialExpressionConvert_5 {
   class=MaterialExpressionConvert
   inputs [
    None {from=MaterialExpressionScalarParameter_15}
    None {from=MaterialExpressionScalarParameter_95}
   ]
   properties {
    convertInputs [
     {
      type=Scalar
      expressionInput {
       expression=&self:MaterialExpressionScalarParameter_15
       inputName=None
       mask=0
       maskA=0
       maskB=0
       maskG=0
       maskR=0
       outputIndex=0
      }
     }
     {
      type=Scalar
      expressionInput {
       expression=&self:MaterialExpressionScalarParameter_95
       inputName=None
       mask=0
       maskA=0
       maskB=0
       maskG=0
       maskR=0
       outputIndex=0
      }
     }
    ]
    convertOutputs [{type=Vector2}]
   }
  }
  MaterialExpressionFloor_0 {class=MaterialExpressionFloor, inputs [None {from=MaterialExpressionAdd_19, output=""}], properties {}}
  MaterialExpressionFloor_1 {class=MaterialExpressionFloor, inputs [None {from=MaterialExpressionAdd_22, output=""}], properties {}}
  MaterialExpressionFloor_2 {class=MaterialExpressionFloor, inputs [None {from=MaterialExpressionAdd_25, output=""}], properties {}}
  MaterialExpressionFloor_3 {class=MaterialExpressionFloor, inputs [None {from=MaterialExpressionComponentMask_0, output=""}], properties {}}
  MaterialExpressionFloor_4 {class=MaterialExpressionFloor, inputs [None {from=MaterialExpressionAdd_28, output=""}], properties {}}
  MaterialExpressionFloor_5 {class=MaterialExpressionFloor, inputs [None {from=MaterialExpressionComponentMask_1, output=""}], properties {}}
  MaterialExpressionFloor_6 {class=MaterialExpressionFloor, inputs [None {from=MaterialExpressionComponentMask_2, output=""}], properties {}}
  MaterialExpressionFrac_0 {class=MaterialExpressionFrac, inputs [None {from=MaterialExpressionMultiply_11, output=""}], properties {}}
  MaterialExpressionFrac_1 {class=MaterialExpressionFrac, inputs [None {from=MaterialExpressionMultiply_36, output=""}], properties {}}
  MaterialExpressionFrac_2 {class=MaterialExpressionFrac, inputs [None {from=MaterialExpressionMultiply_49, output=""}], properties {}}
  MaterialExpressionLinearInterpolate_0 {
   class=MaterialExpressionLinearInterpolate
   inputs [
    A {from=MaterialExpressionMaterialFunctionCall_4, output=Outer}
    B {from=MaterialExpressionMaterialFunctionCall_7, output=Outer}
    Alpha {from=MaterialExpressionScalarParameter_28}
   ]
  }
  MaterialExpressionLinearInterpolate_1 {
   class=MaterialExpressionLinearInterpolate
   inputs [
    A {from=MaterialExpressionMaterialFunctionCall_5, output=Outer}
    B {from=MaterialExpressionMaterialFunctionCall_6, output=Outer}
    Alpha {from=MaterialExpressionScalarParameter_29}
   ]
  }
  MaterialExpressionLinearInterpolate_10 {
   class=MaterialExpressionLinearInterpolate
   inputs [
    A {from=MaterialExpressionAdd_16}
    B {from=MaterialExpressionAppendVector_0}
    Alpha {from=MaterialExpressionReroute_33}
   ]
  }
  MaterialExpressionLinearInterpolate_11 {
   class=MaterialExpressionLinearInterpolate
   inputs [
    A {from=""}
    B {from=""}
    Alpha {from=MaterialExpressionScalarParameter_10}
   ]
   properties {constB=-0.05}
  }
  MaterialExpressionLinearInterpolate_12 {
   class=MaterialExpressionLinearInterpolate
   inputs [
    A {from=MaterialExpressionAdd_32}
    B {from=MaterialExpressionAppendVector_2}
    Alpha {from=MaterialExpressionReroute_50}
   ]
  }
  MaterialExpressionLinearInterpolate_13 {
   class=MaterialExpressionLinearInterpolate
   inputs [
    A {from=""}
    B {from=""}
    Alpha {from=MaterialExpressionScalarParameter_16}
   ]
   properties {constB=-0.075}
  }
  MaterialExpressionLinearInterpolate_14 {
   class=MaterialExpressionLinearInterpolate
   inputs [
    A {from=MaterialExpressionScalarParameter_72}
    B {from=MaterialExpressionNamedRerouteUsage_35}
    Alpha {from=MaterialExpressionMaterialFunctionCall_73, output=And}
   ]
   properties {constA=1}
  }
  MaterialExpressionLinearInterpolate_15 {
   class=MaterialExpressionLinearInterpolate
   inputs [
    A {from=MaterialExpressionScalarParameter_73}
    B {from=MaterialExpressionNamedRerouteUsage_36}
    Alpha {from=MaterialExpressionMaterialFunctionCall_76, output=And}
   ]
   properties {constA=1}
  }
  MaterialExpressionLinearInterpolate_16 {
   class=MaterialExpressionLinearInterpolate
   inputs [
    A {from=MaterialExpressionScalarParameter_74}
    B {from=MaterialExpressionNamedRerouteUsage_37}
    Alpha {from=MaterialExpressionMaterialFunctionCall_79, output=And}
   ]
   properties {constA=1}
  }
  MaterialExpressionLinearInterpolate_17 {
   class=MaterialExpressionLinearInterpolate
   inputs [
    A {from=""}
    B {from=""}
    Alpha {from=MaterialExpressionScalarParameter_68}
   ]
   properties {constA=1, constB=0.5}
  }
  MaterialExpressionLinearInterpolate_18 {
   class=MaterialExpressionLinearInterpolate
   inputs [
    A {from=MaterialExpressionVectorParameter_13, output=RGB}
    B {from=MaterialExpressionVectorParameter_12, output=RGB}
    Alpha {from=MaterialExpressionMaterialFunctionCall_81, output=Smooth}
   ]
  }
  MaterialExpressionLinearInterpolate_19 {
   class=MaterialExpressionLinearInterpolate
   inputs [
    A {from=MaterialExpressionNamedRerouteUsage_32}
    B {from=MaterialExpressionMaterialFunctionCall_17, output=SDF}
    Alpha {from=MaterialExpressionScalarParameter_9}
   ]
  }
  MaterialExpressionLinearInterpolate_2 {
   class=MaterialExpressionLinearInterpolate
   inputs [
    A {from=MaterialExpressionNamedRerouteUsage_31}
    B {from=MaterialExpressionMaterialFunctionCall_16, output=SDF}
    Alpha {from=MaterialExpressionScalarParameter_8}
   ]
  }
  MaterialExpressionLinearInterpolate_20 {
   class=MaterialExpressionLinearInterpolate
   inputs [
    A {from=MaterialExpressionNamedRerouteUsage_33}
    B {from=MaterialExpressionMaterialFunctionCall_18, output=SDF}
    Alpha {from=MaterialExpressionScalarParameter_18}
   ]
  }
  MaterialExpressionLinearInterpolate_21 {
   class=MaterialExpressionLinearInterpolate
   inputs [
    A {from=MaterialExpressionAdd_35}
    B {from=MaterialExpressionAppendVector_4}
    Alpha {from=MaterialExpressionReroute_64}
   ]
  }
  MaterialExpressionLinearInterpolate_23 {
   class=MaterialExpressionLinearInterpolate
   inputs [
    A {from=""}
    B {from=""}
    Alpha {from=MaterialExpressionScalarParameter_17}
   ]
   properties {constB=-0.1}
  }
  MaterialExpressionLinearInterpolate_3 {
   class=MaterialExpressionLinearInterpolate
   inputs [
    A {from=MaterialExpressionMaterialFunctionCall_2, output=SDF}
    B {from=MaterialExpressionReroute_3}
    Alpha {from=MaterialExpressionScalarParameter_67}
   ]
   properties {constAlpha=1}
  }
  MaterialExpressionLinearInterpolate_4 {
   class=MaterialExpressionLinearInterpolate
   inputs [
    A {from=MaterialExpressionMaterialFunctionCall_19, output=Outer}
    B {from=MaterialExpressionMaterialFunctionCall_20, output=Outer}
    Alpha {from=MaterialExpressionScalarParameter_30}
   ]
  }
  MaterialExpressionLinearInterpolate_5 {
   class=MaterialExpressionLinearInterpolate
   inputs [
    A {from=MaterialExpressionMaterialFunctionCall_21, output=Outer}
    B {from=MaterialExpressionMaterialFunctionCall_22, output=Outer}
    Alpha {from=MaterialExpressionScalarParameter_31}
   ]
  }
  MaterialExpressionLinearInterpolate_6 {
   class=MaterialExpressionLinearInterpolate
   inputs [
    A {from=MaterialExpressionMaterialFunctionCall_23, output=Outer}
    B {from=MaterialExpressionMaterialFunctionCall_24, output=Outer}
    Alpha {from=MaterialExpressionScalarParameter_32}
   ]
  }
  MaterialExpressionLinearInterpolate_7 {
   class=MaterialExpressionLinearInterpolate
   inputs [
    A {from=MaterialExpressionNamedRerouteUsage_1}
    B {from=MaterialExpressionNamedRerouteUsage_9}
    Alpha {from=MaterialExpressionScalarParameter_66}
   ]
   properties {constAlpha=1}
  }
  MaterialExpressionLinearInterpolate_8 {
   class=MaterialExpressionLinearInterpolate
   inputs [
    A {from=MaterialExpressionVectorParameter_3, output=RGB}
    B {from=MaterialExpressionVectorParameter_4, output=RGB}
    Alpha {from=MaterialExpressionMax_4}
   ]
  }
  MaterialExpressionLinearInterpolate_9 {
   class=MaterialExpressionLinearInterpolate
   inputs [
    A {from=MaterialExpressionNamedRerouteUsage_3}
    B {from=MaterialExpressionNamedRerouteUsage_13}
    Alpha {from=MaterialExpressionScalarParameter_65}
   ]
  }
  MaterialExpressionMaterialFunctionCall_0 {
   class=MaterialExpressionMaterialFunctionCall
   inputs [
    UV {from=MaterialExpressionMaterialFunctionCall_25, output=UVs}
    Offset {from=MaterialExpressionAdd_2}
    Rotation {from=""}
    Scale {from=MaterialExpressionNamedRerouteUsage_4}
    MirrorY {from=""}
   ]
   properties {materialFunction=/FortUILibrary/MaterialFunctions/Basics/MF_UI_Symmetry_2x1_DEPRECATED}
  }
  MaterialExpressionMaterialFunctionCall_1 {
   class=MaterialExpressionMaterialFunctionCall
   inputs [
    UV {from=MaterialExpressionClamp_3}
    Offset {from=MaterialExpressionAdd_3}
    Rotation {from=""}
    Scale {from=MaterialExpressionNamedRerouteUsage_25}
    MirrorY {from=""}
   ]
   properties {materialFunction=&MF_UI_Symmetry_2x1_DEPRECATED}
  }
  MaterialExpressionMaterialFunctionCall_10 {
   class=MaterialExpressionMaterialFunctionCall
   inputs [
    UVs {from=MaterialExpressionMaterialFunctionCall_69, output=UVs}
    Rotation Center {from=MaterialExpressionReroute_38}
    "Rotation Angle (0-1)" {from=MaterialExpressionReroute_2}
   ]
   properties {materialFunction=&CustomRotator}
  }
  MaterialExpressionMaterialFunctionCall_11 {
   class=MaterialExpressionMaterialFunctionCall
   inputs [MSDF RGB {from=MaterialExpressionTextureSample_2, output=RGB}]
   properties {materialFunction=/FortUILibrary/MaterialFunctions/Shapes/Modifiers/MF_UI_MSDFtoSDF}
  }
  MaterialExpressionMaterialFunctionCall_12 {
   class=MaterialExpressionMaterialFunctionCall
   inputs [
    UV {from=MaterialExpressionMaterialFunctionCall_26, output=UVs}
    Offset {from=MaterialExpressionAdd_7}
    Rotation {from=""}
    Scale {from=MaterialExpressionNamedRerouteUsage_5}
    MirrorY {from=""}
   ]
   properties {materialFunction=&MF_UI_Symmetry_2x1_DEPRECATED}
  }
  MaterialExpressionMaterialFunctionCall_13 {
   class=MaterialExpressionMaterialFunctionCall
   inputs [
    UV {from=MaterialExpressionClamp_0}
    Offset {from=MaterialExpressionAdd_8}
    Rotation {from=""}
    Scale {from=MaterialExpressionConstant2Vector_11}
    MirrorY {from=""}
   ]
   properties {materialFunction=&MF_UI_Symmetry_2x1_DEPRECATED}
  }
  MaterialExpressionMaterialFunctionCall_14 {
   class=MaterialExpressionMaterialFunctionCall
   inputs [
    UV {from=MaterialExpressionClamp_1}
    Offset {from=MaterialExpressionAdd_9}
    Rotation {from=""}
    Scale {from=MaterialExpressionConstant2Vector_10}
    MirrorY {from=""}
   ]
   properties {materialFunction=&MF_UI_Symmetry_2x1_DEPRECATED}
  }
  MaterialExpressionMaterialFunctionCall_15 {
   class=MaterialExpressionMaterialFunctionCall
   inputs [
    UV {from=MaterialExpressionClamp_2}
    Offset {from=MaterialExpressionReroute_5}
    Rotation {from=""}
    Scale {from=MaterialExpressionConstant2Vector_9}
    MirrorY {from=""}
   ]
   properties {materialFunction=&MF_UI_Symmetry_2x1_DEPRECATED}
  }
  MaterialExpressionMaterialFunctionCall_16 {
   class=MaterialExpressionMaterialFunctionCall
   inputs [MSDF RGB {from=MaterialExpressionTextureSample_3, output=RGB}]
   properties {materialFunction=&MF_UI_MSDFtoSDF}
  }
  MaterialExpressionMaterialFunctionCall_17 {
   class=MaterialExpressionMaterialFunctionCall
   inputs [MSDF RGB {from=MaterialExpressionTextureSample_4, output=RGB}]
   properties {materialFunction=&MF_UI_MSDFtoSDF}
  }
  MaterialExpressionMaterialFunctionCall_18 {
   class=MaterialExpressionMaterialFunctionCall
   inputs [MSDF RGB {from=MaterialExpressionTextureSample_5, output=RGB}]
   properties {materialFunction=&MF_UI_MSDFtoSDF}
  }
  MaterialExpressionMaterialFunctionCall_19 {
   class=MaterialExpressionMaterialFunctionCall
   inputs [
    DistanceField {from=MaterialExpressionLinearInterpolate_2}
    Threshold {from=MaterialExpressionScalarParameter_36}
    Smoothing {from=""}
   ]
   properties {materialFunction=&MF_UI_AntiAliasing}
  }
  MaterialExpressionMaterialFunctionCall_2 {
   class=MaterialExpressionMaterialFunctionCall
   inputs [MSDF RGB {from=MaterialExpressionTextureSample_0, output=RGB}]
   properties {materialFunction=&MF_UI_MSDFtoSDF}
  }
  MaterialExpressionMaterialFunctionCall_20 {
   class=MaterialExpressionMaterialFunctionCall
   inputs [
    DistanceField {from=MaterialExpressionReroute_4}
    Threshold {from=MaterialExpressionAdd_11}
    Smoothing {from=""}
   ]
   properties {materialFunction=&MF_UI_AntiAliasing}
  }
  MaterialExpressionMaterialFunctionCall_21 {
   class=MaterialExpressionMaterialFunctionCall
   inputs [
    DistanceField {from=MaterialExpressionLinearInterpolate_19}
    Threshold {from=MaterialExpressionScalarParameter_37}
    Smoothing {from=""}
   ]
   properties {materialFunction=&MF_UI_AntiAliasing}
  }
  MaterialExpressionMaterialFunctionCall_22 {
   class=MaterialExpressionMaterialFunctionCall
   inputs [
    DistanceField {from=MaterialExpressionReroute_6}
    Threshold {from=MaterialExpressionAdd_12}
    Smoothing {from=""}
   ]
   properties {materialFunction=&MF_UI_AntiAliasing}
  }
  MaterialExpressionMaterialFunctionCall_23 {
   class=MaterialExpressionMaterialFunctionCall
   inputs [
    DistanceField {from=MaterialExpressionLinearInterpolate_20}
    Threshold {from=MaterialExpressionScalarParameter_38}
    Smoothing {from=""}
   ]
   properties {materialFunction=&MF_UI_AntiAliasing}
  }
  MaterialExpressionMaterialFunctionCall_24 {
   class=MaterialExpressionMaterialFunctionCall
   inputs [
    DistanceField {from=MaterialExpressionReroute_7}
    Threshold {from=MaterialExpressionAdd_13}
    Smoothing {from=""}
   ]
   properties {materialFunction=&MF_UI_AntiAliasing}
  }
  MaterialExpressionMaterialFunctionCall_25 {
   class=MaterialExpressionMaterialFunctionCall
   inputs [UVs {from=""}]
   properties {materialFunction=/FortUILibrary/MaterialFunctions/Basics/MF_UI_9SliceUVs}
  }
  MaterialExpressionMaterialFunctionCall_26 {
   class=MaterialExpressionMaterialFunctionCall
   inputs [UVs {from=""}]
   properties {materialFunction=&MF_UI_9SliceUVs}
  }
  MaterialExpressionMaterialFunctionCall_27 {
   class=MaterialExpressionMaterialFunctionCall
   inputs [UVs {from=""}]
   properties {materialFunction=&MF_UI_9SliceUVs}
  }
  MaterialExpressionMaterialFunctionCall_28 {
   class=MaterialExpressionMaterialFunctionCall
   inputs [UVs {from=""}]
   properties {materialFunction=&MF_UI_9SliceUVs}
  }
  MaterialExpressionMaterialFunctionCall_29 {
   class=MaterialExpressionMaterialFunctionCall
   inputs [UVs {from=""}]
   properties {materialFunction=&MF_UI_9SliceUVs}
  }
  MaterialExpressionMaterialFunctionCall_3 {
   class=MaterialExpressionMaterialFunctionCall
   inputs [MSDF RGB {from=MaterialExpressionTextureSample_1, output=RGB}]
   properties {materialFunction=&MF_UI_MSDFtoSDF}
  }
  MaterialExpressionMaterialFunctionCall_30 {
   class=MaterialExpressionMaterialFunctionCall
   inputs [UVs {from=""}]
   properties {materialFunction=&MF_UI_9SliceUVs}
  }
  MaterialExpressionMaterialFunctionCall_31 {
   class=MaterialExpressionMaterialFunctionCall
   inputs [
    Input {from=MaterialExpressionScalarParameter_26}
    Input Low {from=MaterialExpressionConstant2Vector_12, output=R}
    Input High {from=MaterialExpressionConstant2Vector_12, output=R}
    Target Low {from=MaterialExpressionConstant2Vector_13, output=R}
    Target High {from=MaterialExpressionConstant2Vector_13, output=R}
   ]
   properties {materialFunction=&RemapValueRange}
  }
  MaterialExpressionMaterialFunctionCall_32 {
   class=MaterialExpressionMaterialFunctionCall
   inputs [Vector or UVs {from=MaterialExpressionMaterialFunctionCall_37, output=Rotated Values}]
   properties {materialFunction=/FortUILibrary/MaterialFunctions/Basics/MF_UI_CheaperVectorToRadial}
  }
  MaterialExpressionMaterialFunctionCall_33 {
   class=MaterialExpressionMaterialFunctionCall
   inputs [
    UVs {from=MaterialExpressionNamedRerouteDeclaration_17}
    Radius {from=MaterialExpressionConstant_3}
   ]
   properties {materialFunction=/FortUILibrary/MaterialFunctions/Shapes/SDFs/MF_SDF_Circle}
  }
  MaterialExpressionMaterialFunctionCall_34 {
   class=MaterialExpressionMaterialFunctionCall
   inputs [
    DistanceField {from=MaterialExpressionMaterialFunctionCall_33, output=SDF}
    Threshold {from=""}
    Smoothing {from=""}
   ]
   properties {materialFunction=&MF_UI_AntiAliasing}
  }
  MaterialExpressionMaterialFunctionCall_35 {
   class=MaterialExpressionMaterialFunctionCall
   inputs [
    UV {from=MaterialExpressionMaterialFunctionCall_50, output=UVs}
    Offset {from=MaterialExpressionMultiply_66}
    Rotation {from=""}
    Scale {from=MaterialExpressionMultiply_67}
    MirrorY {from=""}
   ]
   properties {materialFunction=&MF_UI_Symmetry_2x1_DEPRECATED}
  }
  MaterialExpressionMaterialFunctionCall_36 {
   class=MaterialExpressionMaterialFunctionCall
   inputs [
    UV {from=MaterialExpressionReroute_29}
    Offset {from=MaterialExpressionMultiply_66}
    Rotation {from=""}
    Scale {from=MaterialExpressionMultiply_67}
    MirrorY {from=""}
   ]
   properties {materialFunction=&MF_UI_Symmetry_2x1_DEPRECATED}
  }
  MaterialExpressionMaterialFunctionCall_37 {
   class=MaterialExpressionMaterialFunctionCall
   inputs [
    UVs {from=MaterialExpressionMaterialFunctionCall_35, output=Result}
    Rotation Center {from=""}
    "Rotation Angle (0-1)" {from=MaterialExpressionConstant_21}
   ]
   properties {materialFunction=&CustomRotator}
  }
  MaterialExpressionMaterialFunctionCall_38 {
   class=MaterialExpressionMaterialFunctionCall
   inputs [
    UVs {from=MaterialExpressionMaterialFunctionCall_36, output=Result}
    Rotation Center {from=""}
    "Rotation Angle (0-1)" {from=MaterialExpressionConstant_21}
   ]
   properties {materialFunction=&CustomRotator}
  }
  MaterialExpressionMaterialFunctionCall_39 {
   class=MaterialExpressionMaterialFunctionCall
   inputs [
    UVs {from=MaterialExpressionNamedRerouteUsage_22}
    Radius {from=MaterialExpressionConstant_4}
   ]
   properties {materialFunction=&MF_SDF_Circle}
  }
  MaterialExpressionMaterialFunctionCall_4 {
   class=MaterialExpressionMaterialFunctionCall
   inputs [
    DistanceField {from=MaterialExpressionLinearInterpolate_3}
    Threshold {from=MaterialExpressionScalarParameter_0}
    Smoothing {from=""}
   ]
   properties {materialFunction=&MF_UI_AntiAliasing}
  }
  MaterialExpressionMaterialFunctionCall_40 {
   class=MaterialExpressionMaterialFunctionCall
   inputs [
    DistanceField {from=MaterialExpressionMaterialFunctionCall_39, output=SDF}
    Threshold {from=""}
    Smoothing {from=""}
   ]
   properties {materialFunction=&MF_UI_AntiAliasing}
  }
  MaterialExpressionMaterialFunctionCall_41 {
   class=MaterialExpressionMaterialFunctionCall
   inputs [Vector or UVs {from=MaterialExpressionMaterialFunctionCall_46, output=Rotated Values}]
   properties {materialFunction=&MF_UI_CheaperVectorToRadial}
  }
  MaterialExpressionMaterialFunctionCall_42 {
   class=MaterialExpressionMaterialFunctionCall
   inputs [
    UVs {from=MaterialExpressionMaterialFunctionCall_47, output=Rotated Values}
    Radius {from=MaterialExpressionConstant_2}
   ]
   properties {materialFunction=&MF_SDF_Circle}
  }
  MaterialExpressionMaterialFunctionCall_43 {
   class=MaterialExpressionMaterialFunctionCall
   inputs [
    DistanceField {from=MaterialExpressionMaterialFunctionCall_42, output=SDF}
    Threshold {from=""}
    Smoothing {from=""}
   ]
   properties {materialFunction=&MF_UI_AntiAliasing}
  }
  MaterialExpressionMaterialFunctionCall_44 {
   class=MaterialExpressionMaterialFunctionCall
   inputs [
    UV {from=MaterialExpressionMaterialFunctionCall_49, output=UVs}
    Offset {from=MaterialExpressionConstant2Vector_19}
    Rotation {from=""}
    Scale {from=MaterialExpressionConstant2Vector_18}
    MirrorY {from=""}
   ]
   properties {materialFunction=&MF_UI_Symmetry_2x1_DEPRECATED}
  }
  MaterialExpressionMaterialFunctionCall_45 {
   class=MaterialExpressionMaterialFunctionCall
   inputs [
    UV {from=MaterialExpressionReroute_16}
    Offset {from=MaterialExpressionConstant2Vector_19}
    Rotation {from=""}
    Scale {from=MaterialExpressionConstant2Vector_18}
    MirrorY {from=""}
   ]
   properties {materialFunction=&MF_UI_Symmetry_2x1_DEPRECATED}
  }
  MaterialExpressionMaterialFunctionCall_46 {
   class=MaterialExpressionMaterialFunctionCall
   inputs [
    UVs {from=MaterialExpressionMaterialFunctionCall_44, output=Result}
    Rotation Center {from=""}
    "Rotation Angle (0-1)" {from=MaterialExpressionConstant_1}
   ]
   properties {materialFunction=&CustomRotator}
  }
  MaterialExpressionMaterialFunctionCall_47 {
   class=MaterialExpressionMaterialFunctionCall
   inputs [
    UVs {from=MaterialExpressionMaterialFunctionCall_45, output=Result}
    Rotation Center {from=""}
    "Rotation Angle (0-1)" {from=MaterialExpressionConstant_1}
   ]
   properties {materialFunction=&CustomRotator}
  }
  MaterialExpressionMaterialFunctionCall_48 {
   class=MaterialExpressionMaterialFunctionCall
   inputs [
    Input {from=MaterialExpressionScalarParameter_79}
    Input Low {from=MaterialExpressionConstant2Vector_16, output=R}
    Input High {from=MaterialExpressionConstant2Vector_16, output=R}
    Target Low {from=MaterialExpressionConstant2Vector_17, output=R}
    Target High {from=MaterialExpressionConstant2Vector_17, output=R}
   ]
   properties {materialFunction=&RemapValueRange}
  }
  MaterialExpressionMaterialFunctionCall_49 {
   class=MaterialExpressionMaterialFunctionCall
   inputs [
    UVs {from=MaterialExpressionTextureCoordinate_0}
    Texture Scale {from=""}
   ]
   properties {materialFunction=&ScaleUVsByCenter}
  }
  MaterialExpressionMaterialFunctionCall_5 {
   class=MaterialExpressionMaterialFunctionCall
   inputs [
    DistanceField {from=MaterialExpressionMaterialFunctionCall_3, output=SDF}
    Threshold {from=MaterialExpressionScalarParameter_1}
    Smoothing {from=""}
   ]
   properties {materialFunction=&MF_UI_AntiAliasing}
  }
  MaterialExpressionMaterialFunctionCall_50 {
   class=MaterialExpressionMaterialFunctionCall
   inputs [
    UVs {from=MaterialExpressionTextureCoordinate_3}
    Texture Scale {from=""}
   ]
   properties {materialFunction=&ScaleUVsByCenter}
  }
  MaterialExpressionMaterialFunctionCall_51 {
   class=MaterialExpressionMaterialFunctionCall
   inputs [Vector or UVs {from=MaterialExpressionMaterialFunctionCall_56, output=Rotated Values}]
   properties {materialFunction=&MF_UI_CheaperVectorToRadial}
  }
  MaterialExpressionMaterialFunctionCall_52 {
   class=MaterialExpressionMaterialFunctionCall
   inputs [
    UVs {from=MaterialExpressionMaterialFunctionCall_57, output=Rotated Values}
    Radius {from=MaterialExpressionConstant_7}
   ]
   properties {materialFunction=&MF_SDF_Circle}
  }
  MaterialExpressionMaterialFunctionCall_53 {
   class=MaterialExpressionMaterialFunctionCall
   inputs [
    DistanceField {from=MaterialExpressionMaterialFunctionCall_52, output=SDF}
    Threshold {from=""}
    Smoothing {from=""}
   ]
   properties {materialFunction=&MF_UI_AntiAliasing}
  }
  MaterialExpressionMaterialFunctionCall_54 {
   class=MaterialExpressionMaterialFunctionCall
   inputs [
    UV {from=MaterialExpressionMaterialFunctionCall_59, output=UVs}
    Offset {from=MaterialExpressionConstant2Vector_23}
    Rotation {from=""}
    Scale {from=MaterialExpressionConstant2Vector_22}
    MirrorY {from=""}
   ]
   properties {materialFunction=&MF_UI_Symmetry_2x1_DEPRECATED}
  }
  MaterialExpressionMaterialFunctionCall_55 {
   class=MaterialExpressionMaterialFunctionCall
   inputs [
    UV {from=MaterialExpressionReroute_22}
    Offset {from=MaterialExpressionConstant2Vector_23}
    Rotation {from=""}
    Scale {from=MaterialExpressionConstant2Vector_22}
    MirrorY {from=""}
   ]
   properties {materialFunction=&MF_UI_Symmetry_2x1_DEPRECATED}
  }
  MaterialExpressionMaterialFunctionCall_56 {
   class=MaterialExpressionMaterialFunctionCall
   inputs [
    UVs {from=MaterialExpressionMaterialFunctionCall_54, output=Result}
    Rotation Center {from=""}
    "Rotation Angle (0-1)" {from=MaterialExpressionConstant_6}
   ]
   properties {materialFunction=&CustomRotator}
  }
  MaterialExpressionMaterialFunctionCall_57 {
   class=MaterialExpressionMaterialFunctionCall
   inputs [
    UVs {from=MaterialExpressionMaterialFunctionCall_55, output=Result}
    Rotation Center {from=""}
    "Rotation Angle (0-1)" {from=MaterialExpressionConstant_6}
   ]
   properties {materialFunction=&CustomRotator}
  }
  MaterialExpressionMaterialFunctionCall_58 {
   class=MaterialExpressionMaterialFunctionCall
   inputs [
    Input {from=MaterialExpressionScalarParameter_81}
    Input Low {from=MaterialExpressionConstant2Vector_20, output=R}
    Input High {from=MaterialExpressionConstant2Vector_20, output=R}
    Target Low {from=MaterialExpressionConstant2Vector_21, output=R}
    Target High {from=MaterialExpressionConstant2Vector_21, output=R}
   ]
   properties {materialFunction=&RemapValueRange}
  }
  MaterialExpressionMaterialFunctionCall_59 {
   class=MaterialExpressionMaterialFunctionCall
   inputs [
    UVs {from=MaterialExpressionTextureCoordinate_1}
    Texture Scale {from=""}
   ]
   properties {materialFunction=&ScaleUVsByCenter}
  }
  MaterialExpressionMaterialFunctionCall_6 {
   class=MaterialExpressionMaterialFunctionCall
   inputs [
    DistanceField {from=MaterialExpressionReroute_0}
    Threshold {from=MaterialExpressionAdd_0}
    Smoothing {from=""}
   ]
   properties {materialFunction=&MF_UI_AntiAliasing}
  }
  MaterialExpressionMaterialFunctionCall_60 {
   class=MaterialExpressionMaterialFunctionCall
   inputs [Vector or UVs {from=MaterialExpressionMaterialFunctionCall_65, output=Rotated Values}]
   properties {materialFunction=&MF_UI_CheaperVectorToRadial}
  }
  MaterialExpressionMaterialFunctionCall_61 {
   class=MaterialExpressionMaterialFunctionCall
   inputs [
    UVs {from=MaterialExpressionMaterialFunctionCall_66, output=Rotated Values}
    Radius {from=MaterialExpressionConstant_9}
   ]
   properties {materialFunction=&MF_SDF_Circle}
  }
  MaterialExpressionMaterialFunctionCall_62 {
   class=MaterialExpressionMaterialFunctionCall
   inputs [
    DistanceField {from=MaterialExpressionMaterialFunctionCall_61, output=SDF}
    Threshold {from=""}
    Smoothing {from=""}
   ]
   properties {materialFunction=&MF_UI_AntiAliasing}
  }
  MaterialExpressionMaterialFunctionCall_63 {
   class=MaterialExpressionMaterialFunctionCall
   inputs [
    UV {from=MaterialExpressionMaterialFunctionCall_68, output=UVs}
    Offset {from=MaterialExpressionConstant2Vector_27}
    Rotation {from=""}
    Scale {from=MaterialExpressionConstant2Vector_26}
    MirrorY {from=""}
   ]
   properties {materialFunction=&MF_UI_Symmetry_2x1_DEPRECATED}
  }
  MaterialExpressionMaterialFunctionCall_64 {
   class=MaterialExpressionMaterialFunctionCall
   inputs [
    UV {from=MaterialExpressionReroute_28}
    Offset {from=MaterialExpressionConstant2Vector_27}
    Rotation {from=""}
    Scale {from=MaterialExpressionConstant2Vector_26}
    MirrorY {from=""}
   ]
   properties {materialFunction=&MF_UI_Symmetry_2x1_DEPRECATED}
  }
  MaterialExpressionMaterialFunctionCall_65 {
   class=MaterialExpressionMaterialFunctionCall
   inputs [
    UVs {from=MaterialExpressionMaterialFunctionCall_63, output=Result}
    Rotation Center {from=""}
    "Rotation Angle (0-1)" {from=MaterialExpressionConstant_8}
   ]
   properties {materialFunction=&CustomRotator}
  }
  MaterialExpressionMaterialFunctionCall_66 {
   class=MaterialExpressionMaterialFunctionCall
   inputs [
    UVs {from=MaterialExpressionMaterialFunctionCall_64, output=Result}
    Rotation Center {from=""}
    "Rotation Angle (0-1)" {from=MaterialExpressionConstant_8}
   ]
   properties {materialFunction=&CustomRotator}
  }
  MaterialExpressionMaterialFunctionCall_67 {
   class=MaterialExpressionMaterialFunctionCall
   inputs [
    Input {from=MaterialExpressionScalarParameter_83}
    Input Low {from=MaterialExpressionConstant2Vector_24, output=R}
    Input High {from=MaterialExpressionConstant2Vector_24, output=R}
    Target Low {from=MaterialExpressionConstant2Vector_25, output=R}
    Target High {from=MaterialExpressionConstant2Vector_25, output=R}
   ]
   properties {materialFunction=&RemapValueRange}
  }
  MaterialExpressionMaterialFunctionCall_68 {
   class=MaterialExpressionMaterialFunctionCall
   inputs [
    UVs {from=MaterialExpressionTextureCoordinate_2}
    Texture Scale {from=""}
   ]
   properties {materialFunction=&ScaleUVsByCenter}
  }
  MaterialExpressionMaterialFunctionCall_69 {
   class=MaterialExpressionMaterialFunctionCall
   inputs [
    UVs {from=MaterialExpressionLinearInterpolate_10}
    Texture Scale {from=MaterialExpressionReroute_11}
   ]
   properties {materialFunction=&ScaleUVsByCenter}
  }
  MaterialExpressionMaterialFunctionCall_7 {
   class=MaterialExpressionMaterialFunctionCall
   inputs [
    DistanceField {from=MaterialExpressionReroute_1}
    Threshold {from=MaterialExpressionAdd_1}
    Smoothing {from=""}
   ]
   properties {materialFunction=&MF_UI_AntiAliasing}
  }
  MaterialExpressionMaterialFunctionCall_70 {
   class=MaterialExpressionMaterialFunctionCall
   inputs [Float2 {from=MaterialExpressionAdd_29}]
   properties {materialFunction=&BreakOutFloat2Components}
  }
  MaterialExpressionMaterialFunctionCall_71 {
   class=MaterialExpressionMaterialFunctionCall
   inputs [Float2 {from=MaterialExpressionReroute_47}]
   properties {materialFunction=&BreakOutFloat2Components}
  }
  MaterialExpressionMaterialFunctionCall_72 {
   class=MaterialExpressionMaterialFunctionCall
   inputs [
    A {from=MaterialExpressionScalarParameter_57}
    B {from=MaterialExpressionConstant_13}
   ]
   properties {materialFunction=/FortUILibrary/MaterialFunctions/Basics/MF_UI_BooleanOperations}
  }
  MaterialExpressionMaterialFunctionCall_73 {
   class=MaterialExpressionMaterialFunctionCall
   inputs [
    A {from=MaterialExpressionMaterialFunctionCall_72, output=IsGreaterOrEqual}
    B {from=MaterialExpressionMaterialFunctionCall_74, output=IsLessOrEqual}
   ]
   properties {materialFunction=&MF_UI_BooleanOperations}
  }
  MaterialExpressionMaterialFunctionCall_74 {
   class=MaterialExpressionMaterialFunctionCall
   inputs [
    A {from=MaterialExpressionScalarParameter_86}
    B {from=MaterialExpressionConstant_14}
   ]
   properties {materialFunction=&MF_UI_BooleanOperations}
  }
  MaterialExpressionMaterialFunctionCall_75 {
   class=MaterialExpressionMaterialFunctionCall
   inputs [
    A {from=MaterialExpressionScalarParameter_80}
    B {from=MaterialExpressionConstant_15}
   ]
   properties {materialFunction=&MF_UI_BooleanOperations}
  }
  MaterialExpressionMaterialFunctionCall_76 {
   class=MaterialExpressionMaterialFunctionCall
   inputs [
    A {from=MaterialExpressionMaterialFunctionCall_75, output=IsGreaterOrEqual}
    B {from=MaterialExpressionMaterialFunctionCall_77, output=IsLessOrEqual}
   ]
   properties {materialFunction=&MF_UI_BooleanOperations}
  }
  MaterialExpressionMaterialFunctionCall_77 {
   class=MaterialExpressionMaterialFunctionCall
   inputs [
    A {from=MaterialExpressionScalarParameter_85}
    B {from=MaterialExpressionConstant_16}
   ]
   properties {materialFunction=&MF_UI_BooleanOperations}
  }
  MaterialExpressionMaterialFunctionCall_78 {
   class=MaterialExpressionMaterialFunctionCall
   inputs [
    A {from=MaterialExpressionScalarParameter_82}
    B {from=MaterialExpressionConstant_17}
   ]
   properties {materialFunction=&MF_UI_BooleanOperations}
  }
  MaterialExpressionMaterialFunctionCall_79 {
   class=MaterialExpressionMaterialFunctionCall
   inputs [
    A {from=MaterialExpressionMaterialFunctionCall_78, output=IsGreaterOrEqual}
    B {from=MaterialExpressionMaterialFunctionCall_80, output=IsLessOrEqual}
   ]
   properties {materialFunction=&MF_UI_BooleanOperations}
  }
  MaterialExpressionMaterialFunctionCall_8 {
   class=MaterialExpressionMaterialFunctionCall
   inputs [UVs {from=""}]
   properties {materialFunction=&MF_UI_9SliceUVs}
  }
  MaterialExpressionMaterialFunctionCall_80 {
   class=MaterialExpressionMaterialFunctionCall
   inputs [
    A {from=MaterialExpressionScalarParameter_84}
    B {from=MaterialExpressionConstant_18}
   ]
   properties {materialFunction=&MF_UI_BooleanOperations}
  }
  MaterialExpressionMaterialFunctionCall_81 {
   class=MaterialExpressionMaterialFunctionCall
   inputs [
    UV {from=MaterialExpressionMaterialFunctionCall_82, output=Result}
    GradientStart {from=MaterialExpressionConstant2Vector_28, output=R}
    GradientEnd {from=MaterialExpressionConstant2Vector_28, output=R}
   ]
   properties {materialFunction=&MF_UI_RadialGradient}
  }
  MaterialExpressionMaterialFunctionCall_82 {
   class=MaterialExpressionMaterialFunctionCall
   inputs [
    UV {from=MaterialExpressionMaterialFunctionCall_8, output=UVs}
    Offset {from=MaterialExpressionConstant2Vector_30}
    Rotation {from=MaterialExpressionConstant_19}
    Scale {from=MaterialExpressionMultiply_71}
    MirrorY {from=""}
   ]
   properties {materialFunction=&MF_UI_Symmetry_2x1_DEPRECATED}
  }
  MaterialExpressionMaterialFunctionCall_86 {
   class=MaterialExpressionMaterialFunctionCall
   inputs [UVs {from=""}]
   properties {materialFunction=&MF_UI_9SliceUVs}
  }
  MaterialExpressionMaterialFunctionCall_87 {
   class=MaterialExpressionMaterialFunctionCall
   inputs [
    UVs {from=MaterialExpressionMaterialFunctionCall_88, output=UVs}
    Rotation Center {from=MaterialExpressionReroute_53}
    "Rotation Angle (0-1)" {from=MaterialExpressionReroute_48}
   ]
   properties {materialFunction=&CustomRotator}
  }
  MaterialExpressionMaterialFunctionCall_88 {
   class=MaterialExpressionMaterialFunctionCall
   inputs [
    UVs {from=MaterialExpressionLinearInterpolate_12}
    Texture Scale {from=MaterialExpressionReroute_49}
   ]
   properties {materialFunction=&ScaleUVsByCenter}
  }
  MaterialExpressionMaterialFunctionCall_89 {
   class=MaterialExpressionMaterialFunctionCall
   inputs [Float2 {from=MaterialExpressionAdd_34}]
   properties {materialFunction=&BreakOutFloat2Components}
  }
  MaterialExpressionMaterialFunctionCall_90 {
   class=MaterialExpressionMaterialFunctionCall
   inputs [Float2 {from=MaterialExpressionReroute_57}]
   properties {materialFunction=&BreakOutFloat2Components}
  }
  MaterialExpressionMaterialFunctionCall_91 {
   class=MaterialExpressionMaterialFunctionCall
   inputs [UVs {from=""}]
   properties {materialFunction=&MF_UI_9SliceUVs}
  }
  MaterialExpressionMaterialFunctionCall_92 {
   class=MaterialExpressionMaterialFunctionCall
   inputs [UVs {from=""}]
   properties {materialFunction=&MF_UI_9SliceUVs}
  }
  MaterialExpressionMaterialFunctionCall_93 {
   class=MaterialExpressionMaterialFunctionCall
   inputs [Float2 {from=MaterialExpressionAdd_36}]
   properties {materialFunction=&BreakOutFloat2Components}
  }
  MaterialExpressionMaterialFunctionCall_94 {
   class=MaterialExpressionMaterialFunctionCall
   inputs [Float2 {from=MaterialExpressionReroute_67}]
   properties {materialFunction=&BreakOutFloat2Components}
  }
  MaterialExpressionMaterialFunctionCall_95 {
   class=MaterialExpressionMaterialFunctionCall
   inputs [
    UVs {from=MaterialExpressionMaterialFunctionCall_96, output=UVs}
    Rotation Center {from=MaterialExpressionReroute_71}
    "Rotation Angle (0-1)" {from=MaterialExpressionReroute_69}
   ]
   properties {materialFunction=&CustomRotator}
  }
  MaterialExpressionMaterialFunctionCall_96 {
   class=MaterialExpressionMaterialFunctionCall
   inputs [
    UVs {from=MaterialExpressionLinearInterpolate_21}
    Texture Scale {from=MaterialExpressionReroute_68}
   ]
   properties {materialFunction=&ScaleUVsByCenter}
  }
  MaterialExpressionMax_0 {
   class=MaterialExpressionMax
   inputs [
    A {from=MaterialExpressionLinearInterpolate_7}
    B {from=MaterialExpressionMultiply_3}
   ]
  }
  MaterialExpressionMax_1 {
   class=MaterialExpressionMax
   inputs [
    A {from=MaterialExpressionNamedRerouteUsage_6}
    B {from=MaterialExpressionNamedRerouteUsage_7}
   ]
  }
  MaterialExpressionMax_10 {
   class=MaterialExpressionMax
   inputs [
    A {from=MaterialExpressionMax_0}
    B {from=MaterialExpressionMultiply_5}
   ]
  }
  MaterialExpressionMax_2 {
   class=MaterialExpressionMax
   inputs [
    A {from=MaterialExpressionMax_1}
    B {from=MaterialExpressionNamedRerouteUsage_8}
   ]
  }
  MaterialExpressionMax_3 {
   class=MaterialExpressionMax
   inputs [
    A {from=MaterialExpressionMultiply_9}
    B {from=MaterialExpressionMultiply_7}
   ]
  }
  MaterialExpressionMax_4 {
   class=MaterialExpressionMax
   inputs [
    A {from=MaterialExpressionMultiply_2}
    B {from=MaterialExpressionMultiply_1}
   ]
  }
  MaterialExpressionMax_5 {
   class=MaterialExpressionMax
   inputs [
    A {from=MaterialExpressionNamedRerouteUsage_10}
    B {from=MaterialExpressionNamedRerouteUsage_11}
   ]
  }
  MaterialExpressionMax_6 {
   class=MaterialExpressionMax
   inputs [
    A {from=MaterialExpressionMax_5}
    B {from=MaterialExpressionNamedRerouteUsage_12}
   ]
  }
  MaterialExpressionMax_7 {
   class=MaterialExpressionMax
   inputs [
    A {from=MaterialExpressionMultiply_18}
    B {from=MaterialExpressionMultiply_17}
   ]
  }
  MaterialExpressionMax_8 {
   class=MaterialExpressionMax
   inputs [
    A {from=MaterialExpressionMultiply_24}
    B {from=MaterialExpressionMultiply_23}
   ]
  }
  MaterialExpressionMax_9 {
   class=MaterialExpressionMax
   inputs [
    A {from=MaterialExpressionMultiply_30}
    B {from=MaterialExpressionMultiply_29}
   ]
  }
  MaterialExpressionMultiply_0 {
   class=MaterialExpressionMultiply
   inputs [
    A {from=MaterialExpressionNamedRerouteUsage_40}
    B {from=MaterialExpressionScalarParameter_88}
   ]
  }
  MaterialExpressionMultiply_1 {
   class=MaterialExpressionMultiply
   inputs [
    A {from=MaterialExpressionSmoothStep_0}
    B {from=MaterialExpressionMaterialFunctionCall_40, output=Inner}
   ]
  }
  MaterialExpressionMultiply_10 {
   class=MaterialExpressionMultiply
   inputs [
    A {from=MaterialExpressionConstant2Vector_8}
    B {from=MaterialExpressionMultiply_52}
   ]
  }
  MaterialExpressionMultiply_11 {
   class=MaterialExpressionMultiply
   inputs [
    A {from=MaterialExpressionClamp_4}
    B {from=MaterialExpressionConstant2Vector_3}
   ]
  }
  MaterialExpressionMultiply_12 {
   class=MaterialExpressionMultiply
   inputs [
    A {from=MaterialExpressionReroute_42, output=Scale}
    B {from=MaterialExpressionConstant2Vector_31}
   ]
  }
  MaterialExpressionMultiply_13 {
   class=MaterialExpressionMultiply
   inputs [
    A {from=MaterialExpressionLinearInterpolate_1}
    B {from=MaterialExpressionNamedRerouteUsage_21}
   ]
  }
  MaterialExpressionMultiply_14 {
   class=MaterialExpressionMultiply
   inputs [
    A {from=MaterialExpressionScalarParameter_40}
    B {from=MaterialExpressionSubtract_2}
   ]
  }
  MaterialExpressionMultiply_15 {
   class=MaterialExpressionMultiply
   inputs [
    A {from=MaterialExpressionReroute_13}
    B {from=MaterialExpressionNamedRerouteUsage_23}
   ]
  }
  MaterialExpressionMultiply_16 {
   class=MaterialExpressionMultiply
   inputs [
    A {from=MaterialExpressionVectorParameter_10, output=RGB}
    B {from=MaterialExpressionSubtract_3}
   ]
  }
  MaterialExpressionMultiply_17 {
   class=MaterialExpressionMultiply
   inputs [
    A {from=MaterialExpressionFloor_0}
    B {from=MaterialExpressionMaterialFunctionCall_43, output=Inner}
   ]
  }
  MaterialExpressionMultiply_18 {
   class=MaterialExpressionMultiply
   inputs [
    A {from=MaterialExpressionMaterialFunctionCall_43, output=Inner}
    B {from=""}
   ]
   properties {constB=0}
  }
  MaterialExpressionMultiply_19 {
   class=MaterialExpressionMultiply
   inputs [
    A {from=MaterialExpressionReroute_18}
    B {from=MaterialExpressionNamedRerouteUsage_24}
   ]
  }
  MaterialExpressionMultiply_2 {
   class=MaterialExpressionMultiply
   inputs [
    A {from=MaterialExpressionMaterialFunctionCall_40, output=Inner}
    B {from=""}
   ]
   properties {constB=0}
  }
  MaterialExpressionMultiply_20 {
   class=MaterialExpressionMultiply
   inputs [
    A {from=MaterialExpressionVectorParameter_9, output=RGB}
    B {from=MaterialExpressionSubtract_4}
   ]
  }
  MaterialExpressionMultiply_21 {
   class=MaterialExpressionMultiply
   inputs [
    A {from=MaterialExpressionLinearInterpolate_4}
    B {from=MaterialExpressionNamedRerouteUsage_26}
   ]
  }
  MaterialExpressionMultiply_22 {
   class=MaterialExpressionMultiply
   inputs [
    A {from=MaterialExpressionScalarParameter_50}
    B {from=MaterialExpressionSubtract_5}
   ]
  }
  MaterialExpressionMultiply_23 {
   class=MaterialExpressionMultiply
   inputs [
    A {from=MaterialExpressionFloor_1}
    B {from=MaterialExpressionMaterialFunctionCall_53, output=Inner}
   ]
  }
  MaterialExpressionMultiply_24 {
   class=MaterialExpressionMultiply
   inputs [
    A {from=MaterialExpressionMaterialFunctionCall_53, output=Inner}
    B {from=""}
   ]
   properties {constB=0}
  }
  MaterialExpressionMultiply_25 {
   class=MaterialExpressionMultiply
   inputs [
    A {from=MaterialExpressionReroute_25}
    B {from=MaterialExpressionNamedRerouteUsage_27}
   ]
  }
  MaterialExpressionMultiply_26 {
   class=MaterialExpressionMultiply
   inputs [
    A {from=MaterialExpressionVectorParameter_8, output=RGB}
    B {from=MaterialExpressionSubtract_6}
   ]
  }
  MaterialExpressionMultiply_27 {
   class=MaterialExpressionMultiply
   inputs [
    A {from=MaterialExpressionLinearInterpolate_5}
    B {from=MaterialExpressionNamedRerouteUsage_28}
   ]
  }
  MaterialExpressionMultiply_28 {
   class=MaterialExpressionMultiply
   inputs [
    A {from=MaterialExpressionScalarParameter_51}
    B {from=MaterialExpressionSubtract_7}
   ]
  }
  MaterialExpressionMultiply_29 {
   class=MaterialExpressionMultiply
   inputs [
    A {from=MaterialExpressionFloor_2}
    B {from=MaterialExpressionMaterialFunctionCall_62, output=Inner}
   ]
  }
  MaterialExpressionMultiply_3 {
   class=MaterialExpressionMultiply
   inputs [
    A {from=MaterialExpressionNamedRerouteUsage_0}
    B {from=MaterialExpressionScalarParameter_33}
   ]
  }
  MaterialExpressionMultiply_30 {
   class=MaterialExpressionMultiply
   inputs [
    A {from=MaterialExpressionMaterialFunctionCall_62, output=Inner}
    B {from=""}
   ]
   properties {constB=0}
  }
  MaterialExpressionMultiply_31 {
   class=MaterialExpressionMultiply
   inputs [
    A {from=MaterialExpressionReroute_31}
    B {from=MaterialExpressionNamedRerouteUsage_30}
   ]
  }
  MaterialExpressionMultiply_32 {
   class=MaterialExpressionMultiply
   inputs [
    A {from=MaterialExpressionVectorParameter_7, output=RGB}
    B {from=MaterialExpressionSubtract_8}
   ]
  }
  MaterialExpressionMultiply_33 {
   class=MaterialExpressionMultiply
   inputs [
    A {from=MaterialExpressionLinearInterpolate_6}
    B {from=MaterialExpressionNamedRerouteUsage_29}
   ]
  }
  MaterialExpressionMultiply_34 {
   class=MaterialExpressionMultiply
   inputs [
    A {from=MaterialExpressionScalarParameter_52}
    B {from=MaterialExpressionSubtract_9}
   ]
  }
  MaterialExpressionMultiply_35 {
   class=MaterialExpressionMultiply
   inputs [
    A {from=MaterialExpressionMaterialFunctionCall_71, output=G}
    B {from=""}
   ]
   properties {constB=-1}
  }
  MaterialExpressionMultiply_36 {
   class=MaterialExpressionMultiply
   inputs [
    A {from=MaterialExpressionClamp_5}
    B {from=MaterialExpressionConstant2Vector_34}
   ]
  }
  MaterialExpressionMultiply_37 {
   class=MaterialExpressionMultiply
   inputs [
    A {from=MaterialExpressionReroute_59, output=Scale}
    B {from=MaterialExpressionConstant2Vector_35}
   ]
  }
  MaterialExpressionMultiply_38 {
   class=MaterialExpressionMultiply
   inputs [
    A {from=MaterialExpressionMultiply_19}
    B {from=MaterialExpressionLinearInterpolate_14}
   ]
  }
  MaterialExpressionMultiply_39 {
   class=MaterialExpressionMultiply
   inputs [
    A {from=MaterialExpressionMaterialFunctionCall_90, output=G}
    B {from=""}
   ]
   properties {constB=-1}
  }
  MaterialExpressionMultiply_4 {
   class=MaterialExpressionMultiply
   inputs [
    A {from=MaterialExpressionMax_10}
    B {from=MaterialExpressionScalarParameter_34}
   ]
  }
  MaterialExpressionMultiply_40 {
   class=MaterialExpressionMultiply
   inputs [
    A {from=MaterialExpressionMultiply_25}
    B {from=MaterialExpressionLinearInterpolate_15}
   ]
  }
  MaterialExpressionMultiply_41 {
   class=MaterialExpressionMultiply
   inputs [
    A {from=MaterialExpressionMultiply_31}
    B {from=MaterialExpressionLinearInterpolate_16}
   ]
  }
  MaterialExpressionMultiply_42 {
   class=MaterialExpressionMultiply
   inputs [
    A {from=MaterialExpressionLinearInterpolate_17}
    B {from=MaterialExpressionNamedRerouteUsage_2}
   ]
  }
  MaterialExpressionMultiply_43 {
   class=MaterialExpressionMultiply
   inputs [
    A {from=MaterialExpressionLinearInterpolate_18}
    B {from=MaterialExpressionReroute_34}
   ]
  }
  MaterialExpressionMultiply_44 {
   class=MaterialExpressionMultiply
   inputs [
    A {from=MaterialExpressionReroute_35}
    B {from=MaterialExpressionMultiply_48}
   ]
  }
  MaterialExpressionMultiply_45 {
   class=MaterialExpressionMultiply
   inputs [
    A {from=MaterialExpressionConstant2Vector_6}
    B {from=MaterialExpressionMultiply_54}
   ]
  }
  MaterialExpressionMultiply_46 {
   class=MaterialExpressionMultiply
   inputs [
    A {from=MaterialExpressionScalarParameter_75}
    B {from=MaterialExpressionConstantBiasScale_1}
   ]
  }
  MaterialExpressionMultiply_47 {
   class=MaterialExpressionMultiply
   inputs [
    A {from=MaterialExpressionTime_0}
    B {from=""}
   ]
   properties {constB=0.75}
  }
  MaterialExpressionMultiply_48 {
   class=MaterialExpressionMultiply
   inputs [
    A {from=MaterialExpressionMultiply_46}
    B {from=MaterialExpressionScalarParameter_87}
   ]
  }
  MaterialExpressionMultiply_49 {
   class=MaterialExpressionMultiply
   inputs [
    A {from=MaterialExpressionClamp_6}
    B {from=MaterialExpressionConstant2Vector_37}
   ]
  }
  MaterialExpressionMultiply_5 {
   class=MaterialExpressionMultiply
   inputs [
    A {from=MaterialExpressionNamedRerouteUsage_39}
    B {from=MaterialExpressionScalarParameter_4}
   ]
  }
  MaterialExpressionMultiply_50 {
   class=MaterialExpressionMultiply
   inputs [
    A {from=MaterialExpressionMaterialFunctionCall_94, output=G}
    B {from=""}
   ]
   properties {constB=-1}
  }
  MaterialExpressionMultiply_51 {
   class=MaterialExpressionMultiply
   inputs [
    A {from=MaterialExpressionReroute_74, output=Scale}
    B {from=MaterialExpressionConstant2Vector_38}
   ]
  }
  MaterialExpressionMultiply_52 {
   class=MaterialExpressionMultiply
   inputs [
    A {from=MaterialExpressionScalarParameter_7}
    B {from=MaterialExpressionMultiply_58}
   ]
  }
  MaterialExpressionMultiply_53 {
   class=MaterialExpressionMultiply
   inputs [
    A {from=MaterialExpressionScalarParameter_6}
    B {from=MaterialExpressionMultiply_57}
   ]
  }
  MaterialExpressionMultiply_54 {
   class=MaterialExpressionMultiply
   inputs [
    A {from=MaterialExpressionScalarParameter_5}
    B {from=MaterialExpressionMultiply_55}
   ]
  }
  MaterialExpressionMultiply_55 {
   class=MaterialExpressionMultiply
   inputs [
    A {from=MaterialExpressionScalarParameter_59}
    B {from=MaterialExpressionNamedRerouteUsage_14}
   ]
  }
  MaterialExpressionMultiply_57 {
   class=MaterialExpressionMultiply
   inputs [
    A {from=MaterialExpressionScalarParameter_58}
    B {from=MaterialExpressionNamedRerouteUsage_16}
   ]
  }
  MaterialExpressionMultiply_58 {
   class=MaterialExpressionMultiply
   inputs [
    A {from=MaterialExpressionScalarParameter_24}
    B {from=MaterialExpressionNamedRerouteUsage_19}
   ]
  }
  MaterialExpressionMultiply_61 {
   class=MaterialExpressionMultiply
   inputs [
    A {from=MaterialExpressionScalarParameter_44}
    B {from=""}
   ]
   properties {constB=2}
  }
  MaterialExpressionMultiply_62 {
   class=MaterialExpressionMultiply
   inputs [
    A {from=MaterialExpressionConstant2Vector_5}
    B {from=MaterialExpressionScalarParameter_45}
   ]
  }
  MaterialExpressionMultiply_64 {
   class=MaterialExpressionMultiply
   inputs [
    A {from=MaterialExpressionScalarParameter_47}
    B {from=MaterialExpressionScalarParameter_48}
   ]
  }
  MaterialExpressionMultiply_65 {
   class=MaterialExpressionMultiply
   inputs [
    A {from=MaterialExpressionConstant2Vector_2}
    B {from=MaterialExpressionScalarParameter_49}
   ]
  }
  MaterialExpressionMultiply_66 {
   class=MaterialExpressionMultiply
   inputs [
    A {from=MaterialExpressionConstant2Vector_15}
    B {from=MaterialExpressionScalarParameter_53}
   ]
  }
  MaterialExpressionMultiply_67 {
   class=MaterialExpressionMultiply
   inputs [
    A {from=MaterialExpressionConstant2Vector_14}
    B {from=MaterialExpressionScalarParameter_54}
   ]
  }
  MaterialExpressionMultiply_7 {
   class=MaterialExpressionMultiply
   inputs [
    A {from=MaterialExpressionFloor_4}
    B {from=MaterialExpressionMaterialFunctionCall_34, output=Inner}
   ]
  }
  MaterialExpressionMultiply_71 {
   class=MaterialExpressionMultiply
   inputs [
    A {from=MaterialExpressionConstant2Vector_29}
    B {from=MaterialExpressionScalarParameter_60}
   ]
  }
  MaterialExpressionMultiply_8 {
   class=MaterialExpressionMultiply
   inputs [
    A {from=MaterialExpressionConstant2Vector_7}
    B {from=MaterialExpressionMultiply_53}
   ]
  }
  MaterialExpressionMultiply_9 {
   class=MaterialExpressionMultiply
   inputs [
    A {from=MaterialExpressionMaterialFunctionCall_34, output=Inner}
    B {from=""}
   ]
   properties {constB=0}
  }
  MaterialExpressionNamedRerouteDeclaration_0 {
   class=MaterialExpressionNamedRerouteDeclaration
   inputs [None {from=MaterialExpressionMaterialFunctionCall_4, output=Outer}]
   properties {name=AA'dBracket, nodeColor=<rgba 1 0.8438 0 1>}
  }
  MaterialExpressionNamedRerouteDeclaration_1 {
   class=MaterialExpressionNamedRerouteDeclaration
   inputs [None {from=MaterialExpressionLinearInterpolate_0}]
   properties {name=BracketShadow, nodeColor=<rgba 0.5234 0 1 1>}
  }
  MaterialExpressionNamedRerouteDeclaration_10 {
   class=MaterialExpressionNamedRerouteDeclaration
   inputs [None {from=MaterialExpressionAdd_26}]
   properties {name=AA'dChargedRibbonThree, nodeColor=<rgba 0 0.2031 1 1>}
  }
  MaterialExpressionNamedRerouteDeclaration_11 {
   class=MaterialExpressionNamedRerouteDeclaration
   inputs [None {from=MaterialExpressionAdd_27}]
   properties {name=ChargedRibbonShadowThree, nodeColor=<rgba 0.5234 1 0 1>}
  }
  MaterialExpressionNamedRerouteDeclaration_12 {
   class=MaterialExpressionNamedRerouteDeclaration
   inputs [None {from=MaterialExpressionMax_2}]
   properties {name=ChargedRibbonShadows, nodeColor=<rgba 1 0 0.8437 1>}
  }
  MaterialExpressionNamedRerouteDeclaration_13 {
   class=MaterialExpressionNamedRerouteDeclaration
   inputs [None {from=MaterialExpressionMax_3}]
   properties {name=RadialMask, nodeColor=<rgba 0 0 0 1>}
  }
  MaterialExpressionNamedRerouteDeclaration_14 {
   class=MaterialExpressionNamedRerouteDeclaration
   inputs [None {from=MaterialExpressionLinearInterpolate_8}]
   properties {name=RibbonAttackColor, nodeColor=<rgba 0 0 0 1>}
  }
  MaterialExpressionNamedRerouteDeclaration_15 {
   class=MaterialExpressionNamedRerouteDeclaration
   inputs [None {from=MaterialExpressionMax_6}]
   properties {name=ChargedRibbonArt, nodeColor=<rgba 0.0312 0 1 1>}
  }
  MaterialExpressionNamedRerouteDeclaration_16 {
   class=MaterialExpressionNamedRerouteDeclaration
   inputs [None {from=MaterialExpressionMaterialFunctionCall_32, output=Radial Coordinates}]
   properties {name=RadialUVs, nodeColor=<rgba 0 0 0 1>}
  }
  MaterialExpressionNamedRerouteDeclaration_17 {
   class=MaterialExpressionNamedRerouteDeclaration
   inputs [None {from=MaterialExpressionMaterialFunctionCall_38, output=Rotated Values}]
   properties {name=RotatedCircleUvs, nodeColor=<rgba 0 0 0 1>}
  }
  MaterialExpressionNamedRerouteDeclaration_18 {
   class=MaterialExpressionNamedRerouteDeclaration
   inputs [None {from=MaterialExpressionMax_7}]
   properties {name=RadialMask_ChargedRibbon_One, nodeColor=<rgba 0.2188 0 1 1>}
  }
  MaterialExpressionNamedRerouteDeclaration_19 {
   class=MaterialExpressionNamedRerouteDeclaration
   inputs [None {from=MaterialExpressionMax_8}]
   properties {name=RadialMask_ChargedRibbon_Two, nodeColor=<rgba 0.1016 1 0 1>}
  }
  MaterialExpressionNamedRerouteDeclaration_2 {
   class=MaterialExpressionNamedRerouteDeclaration
   inputs [None {from=MaterialExpressionAdd_18}]
   properties {name=AA'dNormalRibbon, nodeColor=<rgba 0 1 0.2031 1>}
  }
  MaterialExpressionNamedRerouteDeclaration_20 {
   class=MaterialExpressionNamedRerouteDeclaration
   inputs [None {from=MaterialExpressionMax_9}]
   properties {name=RadialMask_ChargedRibbbon_Three, nodeColor=<rgba 1 0 0.4219 1>}
  }
  MaterialExpressionNamedRerouteDeclaration_21 {
   class=MaterialExpressionNamedRerouteDeclaration
   inputs [None {from=MaterialExpressionMultiply_44}]
   properties {name=GlowOpacity, nodeColor=<rgba 1 0 0.3047 1>}
  }
  MaterialExpressionNamedRerouteDeclaration_22 {
   class=MaterialExpressionNamedRerouteDeclaration
   inputs [None {from=MaterialExpressionMultiply_43}]
   properties {name=GlowColor, nodeColor=<rgba 0 0.625 1 1>}
  }
  MaterialExpressionNamedRerouteDeclaration_23 {
   class=MaterialExpressionNamedRerouteDeclaration
   inputs [None {from=MaterialExpressionMaterialFunctionCall_3, output=SDF}]
   properties {name=RegularRibbonSDF, nodeColor=<rgba 0 0.2266 1 1>}
  }
  MaterialExpressionNamedRerouteDeclaration_3 {
   class=MaterialExpressionNamedRerouteDeclaration
   inputs [None {from=MaterialExpressionAdd_17}]
   properties {name=NormalRibbonShadows, nodeColor=<rgba 1 0 0.1172 1>}
  }
  MaterialExpressionNamedRerouteDeclaration_4 {
   class=MaterialExpressionNamedRerouteDeclaration
   inputs [None {from=MaterialExpressionAppendVector_6}]
   properties {name=BracketScale, nodeColor=<rgba 1 0 0.8672 1>}
  }
  MaterialExpressionNamedRerouteDeclaration_5 {
   class=MaterialExpressionNamedRerouteDeclaration
   inputs [None {from=MaterialExpressionMultiply_64}]
   properties {name=RibbonScale, nodeColor=<rgba 1 0 0.2344 1>}
  }
  MaterialExpressionNamedRerouteDeclaration_6 {
   class=MaterialExpressionNamedRerouteDeclaration
   inputs [None {from=MaterialExpressionAdd_20}]
   properties {name=AA'dChargedRibbonOne, nodeColor=<rgba 0.9219 1 0 1>}
  }
  MaterialExpressionNamedRerouteDeclaration_7 {
   class=MaterialExpressionNamedRerouteDeclaration
   inputs [None {from=MaterialExpressionAdd_21}]
   properties {name=ChargedRibbonShadowOne, nodeColor=<rgba 0.7578 0 1 1>}
  }
  MaterialExpressionNamedRerouteDeclaration_8 {
   class=MaterialExpressionNamedRerouteDeclaration
   inputs [None {from=MaterialExpressionAdd_23}]
   properties {name=AA'dChargedRibbonTwo, nodeColor=<rgba 0 1 0.4375 1>}
  }
  MaterialExpressionNamedRerouteDeclaration_9 {
   class=MaterialExpressionNamedRerouteDeclaration
   inputs [None {from=MaterialExpressionAdd_24}]
   properties {name=ChargedRibbonShadowTwo, nodeColor=<rgba 1 0.1172 0 1>}
  }
  MaterialExpressionNamedRerouteUsage_0 {class=MaterialExpressionNamedRerouteUsage}
  MaterialExpressionNamedRerouteUsage_1 {class=MaterialExpressionNamedRerouteUsage}
  MaterialExpressionNamedRerouteUsage_10 {class=MaterialExpressionNamedRerouteUsage}
  MaterialExpressionNamedRerouteUsage_11 {class=MaterialExpressionNamedRerouteUsage}
  MaterialExpressionNamedRerouteUsage_12 {class=MaterialExpressionNamedRerouteUsage}
  MaterialExpressionNamedRerouteUsage_13 {class=MaterialExpressionNamedRerouteUsage}
  MaterialExpressionNamedRerouteUsage_14 {class=MaterialExpressionNamedRerouteUsage}
  MaterialExpressionNamedRerouteUsage_15 {class=MaterialExpressionNamedRerouteUsage}
  MaterialExpressionNamedRerouteUsage_16 {class=MaterialExpressionNamedRerouteUsage}
  MaterialExpressionNamedRerouteUsage_17 {class=MaterialExpressionNamedRerouteUsage}
  MaterialExpressionNamedRerouteUsage_18 {class=MaterialExpressionNamedRerouteUsage}
  MaterialExpressionNamedRerouteUsage_19 {class=MaterialExpressionNamedRerouteUsage}
  MaterialExpressionNamedRerouteUsage_2 {class=MaterialExpressionNamedRerouteUsage}
  MaterialExpressionNamedRerouteUsage_20 {class=MaterialExpressionNamedRerouteUsage}
  MaterialExpressionNamedRerouteUsage_21 {class=MaterialExpressionNamedRerouteUsage}
  MaterialExpressionNamedRerouteUsage_22 {class=MaterialExpressionNamedRerouteUsage}
  MaterialExpressionNamedRerouteUsage_23 {class=MaterialExpressionNamedRerouteUsage}
  MaterialExpressionNamedRerouteUsage_24 {class=MaterialExpressionNamedRerouteUsage}
  MaterialExpressionNamedRerouteUsage_25 {class=MaterialExpressionNamedRerouteUsage}
  MaterialExpressionNamedRerouteUsage_26 {class=MaterialExpressionNamedRerouteUsage}
  MaterialExpressionNamedRerouteUsage_27 {class=MaterialExpressionNamedRerouteUsage}
  MaterialExpressionNamedRerouteUsage_28 {class=MaterialExpressionNamedRerouteUsage}
  MaterialExpressionNamedRerouteUsage_29 {class=MaterialExpressionNamedRerouteUsage}
  MaterialExpressionNamedRerouteUsage_3 {class=MaterialExpressionNamedRerouteUsage}
  MaterialExpressionNamedRerouteUsage_30 {class=MaterialExpressionNamedRerouteUsage}
  MaterialExpressionNamedRerouteUsage_31 {class=MaterialExpressionNamedRerouteUsage}
  MaterialExpressionNamedRerouteUsage_32 {class=MaterialExpressionNamedRerouteUsage}
  MaterialExpressionNamedRerouteUsage_33 {class=MaterialExpressionNamedRerouteUsage}
  MaterialExpressionNamedRerouteUsage_35 {class=MaterialExpressionNamedRerouteUsage}
  MaterialExpressionNamedRerouteUsage_36 {class=MaterialExpressionNamedRerouteUsage}
  MaterialExpressionNamedRerouteUsage_37 {class=MaterialExpressionNamedRerouteUsage}
  MaterialExpressionNamedRerouteUsage_39 {class=MaterialExpressionNamedRerouteUsage}
  MaterialExpressionNamedRerouteUsage_4 {class=MaterialExpressionNamedRerouteUsage}
  MaterialExpressionNamedRerouteUsage_40 {class=MaterialExpressionNamedRerouteUsage}
  MaterialExpressionNamedRerouteUsage_5 {class=MaterialExpressionNamedRerouteUsage}
  MaterialExpressionNamedRerouteUsage_6 {class=MaterialExpressionNamedRerouteUsage}
  MaterialExpressionNamedRerouteUsage_7 {class=MaterialExpressionNamedRerouteUsage}
  MaterialExpressionNamedRerouteUsage_8 {class=MaterialExpressionNamedRerouteUsage}
  MaterialExpressionNamedRerouteUsage_9 {class=MaterialExpressionNamedRerouteUsage}
  MaterialExpressionOneMinus_0 {class=MaterialExpressionOneMinus, inputs [None {from=MaterialExpressionReroute_37}]}
  MaterialExpressionOneMinus_1 {class=MaterialExpressionOneMinus, inputs [None {from=MaterialExpressionMaterialFunctionCall_70, output=G}]}
  MaterialExpressionOneMinus_2 {class=MaterialExpressionOneMinus, inputs [None {from=MaterialExpressionReroute_52}]}
  MaterialExpressionOneMinus_3 {class=MaterialExpressionOneMinus, inputs [None {from=MaterialExpressionMaterialFunctionCall_89, output=G}]}
  MaterialExpressionOneMinus_4 {class=MaterialExpressionOneMinus, inputs [None {from=MaterialExpressionReroute_63}]}
  MaterialExpressionOneMinus_5 {class=MaterialExpressionOneMinus, inputs [None {from=MaterialExpressionMaterialFunctionCall_93, output=G}]}
  MaterialExpressionPinBase_0 {
   class=MaterialExpressionPinBase
   inputs []
   properties {reroutePins [
     Offset {expression=&self:MaterialExpressionReroute_39}
     Scale {expression=&self:MaterialExpressionReroute_42}
     Rotation {expression=&self:MaterialExpressionReroute_40}
     Pivot {expression=&self:MaterialExpressionReroute_41}
    ]}
  }
  MaterialExpressionPinBase_1 {
   class=MaterialExpressionPinBase
   inputs [UVs {from=MaterialExpressionMaterialFunctionCall_10, output=Rotated Values}]
   properties {reroutePins [UVs {expression=&self:MaterialExpressionReroute_45}]}
  }
  MaterialExpressionPinBase_2 {
   class=MaterialExpressionPinBase
   inputs []
   properties {reroutePins [
     Offset {expression=&self:MaterialExpressionReroute_58}
     Scale {expression=&self:MaterialExpressionReroute_59}
     Rotation {expression=&self:MaterialExpressionReroute_60}
     Pivot {expression=&self:MaterialExpressionReroute_61}
    ]}
  }
  MaterialExpressionPinBase_3 {
   class=MaterialExpressionPinBase
   inputs [UVs {from=MaterialExpressionMaterialFunctionCall_87, output=Rotated Values}]
   properties {reroutePins [UVs {expression=&self:MaterialExpressionReroute_62}]}
  }
  MaterialExpressionPinBase_4 {
   class=MaterialExpressionPinBase
   inputs []
   properties {reroutePins [
     Offset {expression=&self:MaterialExpressionReroute_73}
     Scale {expression=&self:MaterialExpressionReroute_74}
     Rotation {expression=&self:MaterialExpressionReroute_75}
     Pivot {expression=&self:MaterialExpressionReroute_76}
    ]}
  }
  MaterialExpressionPinBase_5 {
   class=MaterialExpressionPinBase
   inputs [UVs {from=MaterialExpressionMaterialFunctionCall_95, output=Rotated Values}]
   properties {reroutePins [UVs {expression=&self:MaterialExpressionReroute_77}]}
  }
  MaterialExpressionReroute_0 {class=MaterialExpressionReroute, inputs [None {from=MaterialExpressionMaterialFunctionCall_3, output=SDF}]}
  MaterialExpressionReroute_1 {class=MaterialExpressionReroute, inputs [None {from=MaterialExpressionLinearInterpolate_3}]}
  MaterialExpressionReroute_10 {class=MaterialExpressionReroute, inputs [None {from=MaterialExpressionMaterialFunctionCall_50, output=UVs}]}
  MaterialExpressionReroute_11 {class=MaterialExpressionReroute, inputs [None {from=MaterialExpressionMultiply_12}]}
  MaterialExpressionReroute_12 {class=MaterialExpressionReroute, inputs [None {from=MaterialExpressionMultiply_13}]}
  MaterialExpressionReroute_13 {class=MaterialExpressionReroute, inputs [None {from=MaterialExpressionMaterialFunctionCall_5, output=Outer}]}
  MaterialExpressionReroute_14 {class=MaterialExpressionReroute, inputs [None {from=MaterialExpressionMaterialFunctionCall_41, output=Radial Coordinates}]}
  MaterialExpressionReroute_15 {class=MaterialExpressionReroute, inputs [None {from=MaterialExpressionReroute_14}]}
  MaterialExpressionReroute_16 {class=MaterialExpressionReroute, inputs [None {from=MaterialExpressionReroute_17}]}
  MaterialExpressionReroute_17 {class=MaterialExpressionReroute, inputs [None {from=MaterialExpressionMaterialFunctionCall_49, output=UVs}]}
  MaterialExpressionReroute_18 {class=MaterialExpressionReroute, inputs [None {from=MaterialExpressionMaterialFunctionCall_19, output=Outer}]}
  MaterialExpressionReroute_19 {class=MaterialExpressionReroute, inputs [None {from=MaterialExpressionMultiply_21}]}
  MaterialExpressionReroute_2 {class=MaterialExpressionReroute, inputs [None {from=MaterialExpressionReroute_44}]}
  MaterialExpressionReroute_20 {class=MaterialExpressionReroute, inputs [None {from=MaterialExpressionMaterialFunctionCall_51, output=Radial Coordinates}]}
  MaterialExpressionReroute_21 {class=MaterialExpressionReroute, inputs [None {from=MaterialExpressionReroute_20}]}
  MaterialExpressionReroute_22 {class=MaterialExpressionReroute, inputs [None {from=MaterialExpressionReroute_23}]}
  MaterialExpressionReroute_23 {class=MaterialExpressionReroute, inputs [None {from=MaterialExpressionMaterialFunctionCall_59, output=UVs}]}
  MaterialExpressionReroute_24 {class=MaterialExpressionReroute, inputs [None {from=MaterialExpressionMultiply_27}]}
  MaterialExpressionReroute_25 {class=MaterialExpressionReroute, inputs [None {from=MaterialExpressionMaterialFunctionCall_21, output=Outer}]}
  MaterialExpressionReroute_26 {class=MaterialExpressionReroute, inputs [None {from=MaterialExpressionMaterialFunctionCall_60, output=Radial Coordinates}]}
  MaterialExpressionReroute_27 {class=MaterialExpressionReroute, inputs [None {from=MaterialExpressionReroute_26}]}
  MaterialExpressionReroute_28 {class=MaterialExpressionReroute, inputs [None {from=MaterialExpressionReroute_30}]}
  MaterialExpressionReroute_29 {class=MaterialExpressionReroute, inputs [None {from=MaterialExpressionReroute_10}]}
  MaterialExpressionReroute_3 {class=MaterialExpressionReroute, inputs [None {from=MaterialExpressionMaterialFunctionCall_11, output=SDF}]}
  MaterialExpressionReroute_30 {class=MaterialExpressionReroute, inputs [None {from=MaterialExpressionMaterialFunctionCall_68, output=UVs}]}
  MaterialExpressionReroute_31 {class=MaterialExpressionReroute, inputs [None {from=MaterialExpressionMaterialFunctionCall_23, output=Outer}]}
  MaterialExpressionReroute_32 {class=MaterialExpressionReroute, inputs [None {from=MaterialExpressionMultiply_33}]}
  MaterialExpressionReroute_33 {class=MaterialExpressionReroute, inputs [None {from=MaterialExpressionSaturate_3}]}
  MaterialExpressionReroute_34 {class=MaterialExpressionReroute, inputs [None {from=MaterialExpressionMaterialFunctionCall_81, output=Smooth}]}
  MaterialExpressionReroute_35 {class=MaterialExpressionReroute, inputs [None {from=MaterialExpressionMaterialFunctionCall_81, output=Smooth}]}
  MaterialExpressionReroute_36 {class=MaterialExpressionReroute, inputs [None {from=MaterialExpressionMultiply_11}]}
  MaterialExpressionReroute_37 {class=MaterialExpressionReroute, inputs [None {from=MaterialExpressionFrac_0}]}
  MaterialExpressionReroute_38 {class=MaterialExpressionReroute, inputs [None {from=MaterialExpressionReroute_43}]}
  MaterialExpressionReroute_39 {class=MaterialExpressionReroute, inputs [None {from=MaterialExpressionAdd_8}]}
  MaterialExpressionReroute_4 {class=MaterialExpressionReroute, inputs [None {from=MaterialExpressionLinearInterpolate_2}]}
  MaterialExpressionReroute_40 {class=MaterialExpressionReroute, inputs [None {from=MaterialExpressionLinearInterpolate_11}]}
  MaterialExpressionReroute_41 {class=MaterialExpressionReroute, inputs [None {from=MaterialExpressionConstant2Vector_32}]}
  MaterialExpressionReroute_42 {class=MaterialExpressionReroute, inputs [None {from=MaterialExpressionNamedRerouteUsage_15}]}
  MaterialExpressionReroute_43 {class=MaterialExpressionReroute, inputs [None {from=MaterialExpressionReroute_41, output=Pivot}]}
  MaterialExpressionReroute_44 {class=MaterialExpressionReroute, inputs [None {from=MaterialExpressionReroute_40, output=Rotation}]}
  MaterialExpressionReroute_45 {class=MaterialExpressionReroute, inputs [None {from=MaterialExpressionMaterialFunctionCall_10, output=Rotated Values}]}
  MaterialExpressionReroute_46 {class=MaterialExpressionReroute, inputs [None {from=MaterialExpressionReroute_39, output=Offset}]}
  MaterialExpressionReroute_47 {class=MaterialExpressionReroute, inputs [None {from=MaterialExpressionReroute_39, output=Offset}]}
  MaterialExpressionReroute_48 {class=MaterialExpressionReroute, inputs [None {from=MaterialExpressionReroute_55}]}
  MaterialExpressionReroute_49 {class=MaterialExpressionReroute, inputs [None {from=MaterialExpressionMultiply_37}]}
  MaterialExpressionReroute_5 {class=MaterialExpressionReroute, inputs [None {from=MaterialExpressionAdd_10}]}
  MaterialExpressionReroute_50 {class=MaterialExpressionReroute, inputs [None {from=MaterialExpressionSaturate_4}]}
  MaterialExpressionReroute_51 {class=MaterialExpressionReroute, inputs [None {from=MaterialExpressionMultiply_36}]}
  MaterialExpressionReroute_52 {class=MaterialExpressionReroute, inputs [None {from=MaterialExpressionFrac_1}]}
  MaterialExpressionReroute_53 {class=MaterialExpressionReroute, inputs [None {from=MaterialExpressionReroute_54}]}
  MaterialExpressionReroute_54 {class=MaterialExpressionReroute, inputs [None {from=MaterialExpressionReroute_61, output=Pivot}]}
  MaterialExpressionReroute_55 {class=MaterialExpressionReroute, inputs [None {from=MaterialExpressionReroute_60, output=Rotation}]}
  MaterialExpressionReroute_56 {class=MaterialExpressionReroute, inputs [None {from=MaterialExpressionReroute_58, output=Offset}]}
  MaterialExpressionReroute_57 {class=MaterialExpressionReroute, inputs [None {from=MaterialExpressionReroute_58, output=Offset}]}
  MaterialExpressionReroute_58 {class=MaterialExpressionReroute, inputs [None {from=MaterialExpressionAdd_9}]}
  MaterialExpressionReroute_59 {class=MaterialExpressionReroute, inputs [None {from=MaterialExpressionNamedRerouteUsage_17}]}
  MaterialExpressionReroute_6 {class=MaterialExpressionReroute, inputs [None {from=MaterialExpressionLinearInterpolate_19}]}
  MaterialExpressionReroute_60 {class=MaterialExpressionReroute, inputs [None {from=MaterialExpressionLinearInterpolate_13}]}
  MaterialExpressionReroute_61 {class=MaterialExpressionReroute, inputs [None {from=MaterialExpressionConstant2Vector_36}]}
  MaterialExpressionReroute_62 {class=MaterialExpressionReroute, inputs [None {from=MaterialExpressionMaterialFunctionCall_87, output=Rotated Values}]}
  MaterialExpressionReroute_63 {class=MaterialExpressionReroute, inputs [None {from=MaterialExpressionFrac_2}]}
  MaterialExpressionReroute_64 {class=MaterialExpressionReroute, inputs [None {from=MaterialExpressionSaturate_5}]}
  MaterialExpressionReroute_65 {class=MaterialExpressionReroute, inputs [None {from=MaterialExpressionMultiply_49}]}
  MaterialExpressionReroute_66 {class=MaterialExpressionReroute, inputs [None {from=MaterialExpressionReroute_73, output=Offset}]}
  MaterialExpressionReroute_67 {class=MaterialExpressionReroute, inputs [None {from=MaterialExpressionReroute_73, output=Offset}]}
  MaterialExpressionReroute_68 {class=MaterialExpressionReroute, inputs [None {from=MaterialExpressionMultiply_51}]}
  MaterialExpressionReroute_69 {class=MaterialExpressionReroute, inputs [None {from=MaterialExpressionReroute_70}]}
  MaterialExpressionReroute_7 {class=MaterialExpressionReroute, inputs [None {from=MaterialExpressionLinearInterpolate_20}]}
  MaterialExpressionReroute_70 {class=MaterialExpressionReroute, inputs [None {from=MaterialExpressionReroute_75, output=Rotation}]}
  MaterialExpressionReroute_71 {class=MaterialExpressionReroute, inputs [None {from=MaterialExpressionReroute_72}]}
  MaterialExpressionReroute_72 {class=MaterialExpressionReroute, inputs [None {from=MaterialExpressionReroute_76, output=Pivot}]}
  MaterialExpressionReroute_73 {class=MaterialExpressionReroute, inputs [None {from=MaterialExpressionReroute_5}]}
  MaterialExpressionReroute_74 {class=MaterialExpressionReroute, inputs [None {from=MaterialExpressionNamedRerouteUsage_20}]}
  MaterialExpressionReroute_75 {class=MaterialExpressionReroute, inputs [None {from=MaterialExpressionLinearInterpolate_23}]}
  MaterialExpressionReroute_76 {class=MaterialExpressionReroute, inputs [None {from=MaterialExpressionConstant2Vector_39}]}
  MaterialExpressionReroute_77 {class=MaterialExpressionReroute, inputs [None {from=MaterialExpressionMaterialFunctionCall_95, output=Rotated Values}]}
  MaterialExpressionReroute_8 {class=MaterialExpressionReroute, inputs [None {from=MaterialExpressionNamedRerouteDeclaration_16}]}
  MaterialExpressionReroute_9 {class=MaterialExpressionReroute, inputs [None {from=MaterialExpressionReroute_8}]}
  MaterialExpressionSaturate_0 {class=MaterialExpressionSaturate, inputs [None {from=MaterialExpressionMaterialFunctionCall_48, output=Result}]}
  MaterialExpressionSaturate_1 {class=MaterialExpressionSaturate, inputs [None {from=MaterialExpressionMaterialFunctionCall_58, output=Result}]}
  MaterialExpressionSaturate_2 {class=MaterialExpressionSaturate, inputs [None {from=MaterialExpressionMaterialFunctionCall_67, output=Result}]}
  MaterialExpressionSaturate_3 {class=MaterialExpressionSaturate, inputs [None {from=MaterialExpressionFloor_3}]}
  MaterialExpressionSaturate_4 {class=MaterialExpressionSaturate, inputs [None {from=MaterialExpressionFloor_5}]}
  MaterialExpressionSaturate_5 {class=MaterialExpressionSaturate, inputs [None {from=MaterialExpressionFloor_6}]}
  MaterialExpressionScalarParameter_0 {class=MaterialExpressionScalarParameter, properties {defaultValue=0.22, parameterName=Threshold}}
  MaterialExpressionScalarParameter_1 {class=MaterialExpressionScalarParameter, properties {defaultValue=0.22, parameterName=Threshold}}
  MaterialExpressionScalarParameter_10 {class=MaterialExpressionScalarParameter, properties {parameterName=RibbonRotation}}
  MaterialExpressionScalarParameter_11 {class=MaterialExpressionScalarParameter, properties {defaultValue=-0.304, parameterName=BracketOffsetX}}
  MaterialExpressionScalarParameter_12 {class=MaterialExpressionScalarParameter, properties {defaultValue=-0.304, parameterName=BracketOffsetX}}
  MaterialExpressionScalarParameter_13 {class=MaterialExpressionScalarParameter, properties {defaultValue=-0.304, parameterName=BracketOffsetX}}
  MaterialExpressionScalarParameter_14 {class=MaterialExpressionScalarParameter, properties {defaultValue=-0.304, parameterName=BracketOffsetX}}
  MaterialExpressionScalarParameter_15 {class=MaterialExpressionScalarParameter, properties {defaultValue=-0.304, parameterName=BracketOffsetX}}
  MaterialExpressionScalarParameter_16 {class=MaterialExpressionScalarParameter, properties {parameterName=RibbonRotation}}
  MaterialExpressionScalarParameter_17 {class=MaterialExpressionScalarParameter, properties {parameterName=RibbonRotation}}
  MaterialExpressionScalarParameter_18 {class=MaterialExpressionScalarParameter, properties {parameterName=IsLeaping, sliderMax=1}}
  MaterialExpressionScalarParameter_19 {class=MaterialExpressionScalarParameter, properties {defaultValue=0.24, group=Ribbon, parameterName=RibbonThreshold}}
  MaterialExpressionScalarParameter_2 {class=MaterialExpressionScalarParameter, properties {defaultValue=0.22, parameterName=Threshold}}
  MaterialExpressionScalarParameter_21 {class=MaterialExpressionScalarParameter, properties {defaultValue=0.24, group=Ribbon, parameterName=RibbonThreshold}}
  MaterialExpressionScalarParameter_23 {class=MaterialExpressionScalarParameter, properties {defaultValue=0.24, group=Ribbon, parameterName=RibbonThreshold}}
  MaterialExpressionScalarParameter_24 {class=MaterialExpressionScalarParameter, properties {defaultValue=1, parameterName=RibbonGap}}
  MaterialExpressionScalarParameter_26 {
   class=MaterialExpressionScalarParameter
   properties {defaultValue=1, desc=Used for the Leap Cooldown on Reticle, group=Fill, parameterName=LeapCooldown, sliderMax=1}
  }
  MaterialExpressionScalarParameter_27 {class=MaterialExpressionScalarParameter, properties {defaultValue=-0.304, parameterName=BracketOffsetX}}
  MaterialExpressionScalarParameter_28 {class=MaterialExpressionScalarParameter, properties {defaultValue=0.5, group=Opacity, parameterName=ShadowOpacity}}
  MaterialExpressionScalarParameter_29 {class=MaterialExpressionScalarParameter, properties {defaultValue=0.5, group=Opacity, parameterName=ShadowOpacity}}
  MaterialExpressionScalarParameter_3 {class=MaterialExpressionScalarParameter, properties {defaultValue=0.22, parameterName=Threshold}}
  MaterialExpressionScalarParameter_30 {class=MaterialExpressionScalarParameter, properties {defaultValue=0.5, group=Opacity, parameterName=ShadowOpacity}}
  MaterialExpressionScalarParameter_31 {class=MaterialExpressionScalarParameter, properties {defaultValue=0.5, group=Opacity, parameterName=ShadowOpacity}}
  MaterialExpressionScalarParameter_32 {class=MaterialExpressionScalarParameter, properties {defaultValue=0.5, group=Opacity, parameterName=ShadowOpacity}}
  MaterialExpressionScalarParameter_33 {class=MaterialExpressionScalarParameter, properties {defaultValue=1, group=Opacity, parameterName=BracketOpacity}}
  MaterialExpressionScalarParameter_34 {class=MaterialExpressionScalarParameter, properties {defaultValue=0.85, group=Opacity, parameterName=MasterOpacity}}
  MaterialExpressionScalarParameter_36 {class=MaterialExpressionScalarParameter, properties {defaultValue=0.24, group=Ribbon, parameterName=RibbonThreshold}}
  MaterialExpressionScalarParameter_37 {class=MaterialExpressionScalarParameter, properties {defaultValue=0.24, group=Ribbon, parameterName=RibbonThreshold}}
  MaterialExpressionScalarParameter_38 {class=MaterialExpressionScalarParameter, properties {defaultValue=0.24, group=Ribbon, parameterName=RibbonThreshold}}
  MaterialExpressionScalarParameter_4 {class=MaterialExpressionScalarParameter, properties {desc=Used for Leap Charge, group=Fill, parameterName=LeapChargeFill}}
  MaterialExpressionScalarParameter_40 {class=MaterialExpressionScalarParameter, properties {defaultValue=1, group=Opacity, parameterName=RibbonBackgroundOpacity}}
  MaterialExpressionScalarParameter_44 {class=MaterialExpressionScalarParameter, properties {defaultValue=0.7, parameterName=ReticleScale}}
  MaterialExpressionScalarParameter_45 {class=MaterialExpressionScalarParameter, properties {defaultValue=0.7, parameterName=ReticleScale}}
  MaterialExpressionScalarParameter_47 {class=MaterialExpressionScalarParameter, properties {defaultValue=1, parameterName=RibbonScale}}
  MaterialExpressionScalarParameter_48 {class=MaterialExpressionScalarParameter, properties {defaultValue=0.7, parameterName=ReticleScale}}
  MaterialExpressionScalarParameter_49 {class=MaterialExpressionScalarParameter, properties {defaultValue=0.7, parameterName=ReticleScale}}
  MaterialExpressionScalarParameter_5 {class=MaterialExpressionScalarParameter, properties {parameterName=IsLeaping}}
  MaterialExpressionScalarParameter_50 {class=MaterialExpressionScalarParameter, properties {defaultValue=1, group=Opacity, parameterName=RibbonBackgroundOpacity}}
  MaterialExpressionScalarParameter_51 {class=MaterialExpressionScalarParameter, properties {defaultValue=1, group=Opacity, parameterName=RibbonBackgroundOpacity}}
  MaterialExpressionScalarParameter_52 {class=MaterialExpressionScalarParameter, properties {defaultValue=1, group=Opacity, parameterName=RibbonBackgroundOpacity}}
  MaterialExpressionScalarParameter_53 {class=MaterialExpressionScalarParameter, properties {defaultValue=0.7, parameterName=ReticleScale}}
  MaterialExpressionScalarParameter_54 {class=MaterialExpressionScalarParameter, properties {defaultValue=0.7, parameterName=ReticleScale}}
  MaterialExpressionScalarParameter_57 {class=MaterialExpressionScalarParameter, properties {desc=Used for Leap Charge, group=Fill, parameterName=LeapChargeFill}}
  MaterialExpressionScalarParameter_58 {class=MaterialExpressionScalarParameter, properties {defaultValue=1, parameterName=RibbonGap}}
  MaterialExpressionScalarParameter_59 {class=MaterialExpressionScalarParameter, properties {defaultValue=1, parameterName=RibbonGap}}
  MaterialExpressionScalarParameter_6 {class=MaterialExpressionScalarParameter, properties {parameterName=IsLeaping}}
  MaterialExpressionScalarParameter_60 {class=MaterialExpressionScalarParameter, properties {defaultValue=0.7, parameterName=ReticleScale}}
  MaterialExpressionScalarParameter_65 {class=MaterialExpressionScalarParameter, properties {parameterName=IsLeaping, sliderMax=1.5}}
  MaterialExpressionScalarParameter_66 {class=MaterialExpressionScalarParameter, properties {parameterName=IsLeaping, sliderMax=1}}
  MaterialExpressionScalarParameter_67 {class=MaterialExpressionScalarParameter, properties {parameterName=IsLeaping, sliderMax=1}}
  MaterialExpressionScalarParameter_68 {class=MaterialExpressionScalarParameter, properties {parameterName=IsLeaping, sliderMax=1}}
  MaterialExpressionScalarParameter_7 {class=MaterialExpressionScalarParameter, properties {parameterName=IsLeaping}}
  MaterialExpressionScalarParameter_72 {class=MaterialExpressionScalarParameter, properties {defaultValue=1, group=Ribbon, parameterName=IdleChargeRibbonColor}}
  MaterialExpressionScalarParameter_73 {class=MaterialExpressionScalarParameter, properties {defaultValue=1, group=Ribbon, parameterName=IdleChargeRibbonColor}}
  MaterialExpressionScalarParameter_74 {class=MaterialExpressionScalarParameter, properties {defaultValue=1, group=Ribbon, parameterName=IdleChargeRibbonColor}}
  MaterialExpressionScalarParameter_75 {class=MaterialExpressionScalarParameter, properties {defaultValue=2, group=Opacity, parameterName=GlowOpacity}}
  MaterialExpressionScalarParameter_79 {class=MaterialExpressionScalarParameter, properties {desc=Used for Leap Charge, group=Fill, parameterName=LeapChargeFill}}
  MaterialExpressionScalarParameter_8 {class=MaterialExpressionScalarParameter, properties {parameterName=IsLeaping, sliderMax=1}}
  MaterialExpressionScalarParameter_80 {class=MaterialExpressionScalarParameter, properties {desc=Used for Leap Charge, group=Fill, parameterName=LeapChargeFill}}
  MaterialExpressionScalarParameter_81 {class=MaterialExpressionScalarParameter, properties {desc=Used for Leap Charge, group=Fill, parameterName=LeapChargeFill}}
  MaterialExpressionScalarParameter_82 {class=MaterialExpressionScalarParameter, properties {desc=Used for Leap Charge, group=Fill, parameterName=LeapChargeFill}}
  MaterialExpressionScalarParameter_83 {class=MaterialExpressionScalarParameter, properties {desc=Used for Leap Charge, group=Fill, parameterName=LeapChargeFill}}
  MaterialExpressionScalarParameter_84 {class=MaterialExpressionScalarParameter, properties {desc=Used for Leap Charge, group=Fill, parameterName=LeapChargeFill}}
  MaterialExpressionScalarParameter_85 {class=MaterialExpressionScalarParameter, properties {desc=Used for Leap Charge, group=Fill, parameterName=LeapChargeFill}}
  MaterialExpressionScalarParameter_86 {class=MaterialExpressionScalarParameter, properties {desc=Used for Leap Charge, group=Fill, parameterName=LeapChargeFill}}
  MaterialExpressionScalarParameter_87 {class=MaterialExpressionScalarParameter, properties {desc=Used for Leap Charge, group=Fill, parameterName=LeapChargeFill}}
  MaterialExpressionScalarParameter_88 {class=MaterialExpressionScalarParameter, properties {desc=Used for Leap Charge, group=Fill, parameterName=LeapChargeFill}}
  MaterialExpressionScalarParameter_89 {class=MaterialExpressionScalarParameter, properties {parameterName=BracketOffsetY}}
  MaterialExpressionScalarParameter_9 {class=MaterialExpressionScalarParameter, properties {parameterName=IsLeaping, sliderMax=1}}
  MaterialExpressionScalarParameter_90 {class=MaterialExpressionScalarParameter, properties {parameterName=BracketOffsetY}}
  MaterialExpressionScalarParameter_91 {class=MaterialExpressionScalarParameter, properties {parameterName=BracketOffsetY}}
  MaterialExpressionScalarParameter_93 {class=MaterialExpressionScalarParameter, properties {parameterName=BracketOffsetY}}
  MaterialExpressionScalarParameter_94 {class=MaterialExpressionScalarParameter, properties {parameterName=BracketOffsetY}}
  MaterialExpressionScalarParameter_95 {class=MaterialExpressionScalarParameter, properties {parameterName=BracketOffsetY}}
  MaterialExpressionSine_0 {class=MaterialExpressionSine, inputs [None {from=MaterialExpressionMultiply_47}]}
  MaterialExpressionSmoothStep_0 {
   class=MaterialExpressionSmoothStep
   inputs [
    Min {from="", output=""}
    Max {from="", output=""}
    Value {from=MaterialExpressionAdd_14, output=""}
   ]
   properties {constMax=1.3, constMin=1.01}
  }
  MaterialExpressionSubtract_2 {
   class=MaterialExpressionSubtract
   inputs [
    A {from=MaterialExpressionLinearInterpolate_1}
    B {from=MaterialExpressionMultiply_13}
   ]
  }
  MaterialExpressionSubtract_3 {
   class=MaterialExpressionSubtract
   inputs [
    A {from=MaterialExpressionReroute_13}
    B {from=MaterialExpressionMultiply_15}
   ]
  }
  MaterialExpressionSubtract_4 {
   class=MaterialExpressionSubtract
   inputs [
    A {from=MaterialExpressionReroute_18}
    B {from=MaterialExpressionMultiply_19}
   ]
  }
  MaterialExpressionSubtract_5 {
   class=MaterialExpressionSubtract
   inputs [
    A {from=MaterialExpressionLinearInterpolate_4}
    B {from=MaterialExpressionMultiply_21}
   ]
  }
  MaterialExpressionSubtract_6 {
   class=MaterialExpressionSubtract
   inputs [
    A {from=MaterialExpressionReroute_25}
    B {from=MaterialExpressionMultiply_25}
   ]
  }
  MaterialExpressionSubtract_7 {
   class=MaterialExpressionSubtract
   inputs [
    A {from=MaterialExpressionLinearInterpolate_5}
    B {from=MaterialExpressionMultiply_27}
   ]
  }
  MaterialExpressionSubtract_8 {
   class=MaterialExpressionSubtract
   inputs [
    A {from=MaterialExpressionReroute_31}
    B {from=MaterialExpressionMultiply_31}
   ]
  }
  MaterialExpressionSubtract_9 {
   class=MaterialExpressionSubtract
   inputs [
    A {from=MaterialExpressionLinearInterpolate_6}
    B {from=MaterialExpressionMultiply_33}
   ]
  }
  MaterialExpressionTextureCoordinate_0 {class=MaterialExpressionTextureCoordinate}
  MaterialExpressionTextureCoordinate_1 {class=MaterialExpressionTextureCoordinate}
  MaterialExpressionTextureCoordinate_2 {class=MaterialExpressionTextureCoordinate}
  MaterialExpressionTextureCoordinate_3 {class=MaterialExpressionTextureCoordinate}
  MaterialExpressionTextureSample_0 {
   class=MaterialExpressionTextureSample
   inputs [
    UVs {from=MaterialExpressionMaterialFunctionCall_0, output=Result}
    Tex {from=""}
    Apply View MipBias {from=""}
   ]
   properties {texture=/BeastGirlStormClawsUI/Reticle/Textures/T_UI_BentBaton_Bracket_Normal}
  }
  MaterialExpressionTextureSample_1 {
   class=MaterialExpressionTextureSample
   inputs [
    UVs {from=MaterialExpressionMaterialFunctionCall_1, output=Result}
    Tex {from=""}
    Apply View MipBias {from=""}
   ]
   properties {texture=/BeastGirlStormClawsUI/Reticle/Textures/T_UI_BentBaton_Charge_Normal}
  }
  MaterialExpressionTextureSample_2 {
   class=MaterialExpressionTextureSample
   inputs [
    UVs {from=MaterialExpressionMaterialFunctionCall_12, output=Result}
    Tex {from=""}
    Apply View MipBias {from=""}
   ]
   properties {texture=/BeastGirlStormClawsUI/Reticle/Textures/T_UI_BentBaton_Bracket_Charging}
  }
  MaterialExpressionTextureSample_3 {
   class=MaterialExpressionTextureSample
   inputs [
    UVs {from=MaterialExpressionReroute_45, output=UVs}
    Tex {from=""}
    Apply View MipBias {from=""}
   ]
   properties {samplerSource=SSM_Clamp_WorldGroupSettings, texture=/BeastGirlStormClawsUI/Reticle/Textures/T_UI_BentBaton_Ribbon_1}
  }
  MaterialExpressionTextureSample_4 {
   class=MaterialExpressionTextureSample
   inputs [
    UVs {from=MaterialExpressionReroute_62, output=UVs}
    Tex {from=""}
    Apply View MipBias {from=""}
   ]
   properties {texture=/BeastGirlStormClawsUI/Reticle/Textures/T_UI_BentBaton_Ribbon_2}
  }
  MaterialExpressionTextureSample_5 {
   class=MaterialExpressionTextureSample
   inputs [
    UVs {from=MaterialExpressionReroute_77, output=UVs}
    Tex {from=""}
    Apply View MipBias {from=""}
   ]
   properties {texture=/BeastGirlStormClawsUI/Reticle/Textures/T_UI_BentBaton_Ribbon_3}
  }
  MaterialExpressionTime_0 {class=MaterialExpressionTime}
  MaterialExpressionVectorParameter_10 {class=MaterialExpressionVectorParameter, properties {parameterName=RibbonBackingColor, defaultValue=<rgba 1 1 1 1>}}
  MaterialExpressionVectorParameter_12 {class=MaterialExpressionVectorParameter, properties {parameterName=RibbonTopColor, defaultValue=<rgba 1 0.7932 0.1927 1>}}
  MaterialExpressionVectorParameter_13 {class=MaterialExpressionVectorParameter, properties {parameterName=RibbonBottomColor, defaultValue=<rgba 0.7188 0.0132 0.3225 1>}}
  MaterialExpressionVectorParameter_3 {class=MaterialExpressionVectorParameter, properties {parameterName=RibbonTopColor, defaultValue=<rgba 1 0.7932 0.1927 1>}}
  MaterialExpressionVectorParameter_4 {class=MaterialExpressionVectorParameter, properties {parameterName=RibbonBottomColor, defaultValue=<rgba 0.7188 0.0132 0.3225 1>}}
  MaterialExpressionVectorParameter_7 {class=MaterialExpressionVectorParameter, properties {parameterName=RibbonBackingColor, defaultValue=<rgba 1 1 1 1>}}
  MaterialExpressionVectorParameter_8 {class=MaterialExpressionVectorParameter, properties {parameterName=RibbonBackingColor, defaultValue=<rgba 1 1 1 1>}}
  MaterialExpressionVectorParameter_9 {class=MaterialExpressionVectorParameter, properties {parameterName=RibbonBackingColor, defaultValue=<rgba 1 1 1 1>}}
 }
 properties {bAutomaticallySetUsageInEditor=false, blendMode=BLEND_Translucent, materialDomain=MD_UI, usedShadingModels=Default Lit}
}
