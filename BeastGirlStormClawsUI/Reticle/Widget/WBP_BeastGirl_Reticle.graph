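# Weapon reticle widget for BeastGirl Storm Claws, showing air dive, Storm Slam charge and cooldown state.
# Async gameplay tag listeners on the owning PlayerController toggle visibility, dive animations and cooldown state.
# A charge-started event supplies charge count and duration; looping timers derive charge fill from server time and cooldown from active gameplay effects.
# Scalar parameters and animation tracks drive one dynamic reticle material, with fill split per max charge.
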WidgetBlueprint /BeastGirlStormClawsUI/Reticle/Widget/WBP_BeastGirl_Reticle
animations [
 Anim_Equip {
  binding_count=1
  length_seconds=0.35
  object=&self:Anim_Equip
  sequence {bindings [Image_Reticle {
     id=#1
     possessed_class=Image
     tracks [{
       type=MovieSceneWidgetMaterialTrack
       sections [{
         type=MovieSceneParameterSection
         end_frame=7
         start_frame=0
         channels [MasterOpacity {type=MovieSceneScriptingFloatChannel, keys [[0,0],[5,0.8]]}]
        }]
      }]
    }]}
 }
 Anim_CooldownComplete {
  binding_count=1
  length_seconds=0.25
  object=&self:Anim_CooldownComplete
  sequence {bindings [Image_Reticle {
     id=#2
     possessed_class=Image
     tracks [{
       type=MovieSceneWidgetMaterialTrack
       sections [{
         type=MovieSceneParameterSection
         end_frame=5
         start_frame=0
         channels [Flash {type=MovieSceneScriptingFloatChannel, keys [[0,0],[2.4,1],[5,0]]}]
        }]
      }]
    }]}
 }
 Anim_AirDiveLoop {
  binding_count=1
  length_seconds=0.5
  object=&self:Anim_AirDiveLoop
  sequence {bindings [Image_Reticle {
     id=#3
     possessed_class=Image
     tracks [{
       type=MovieSceneWidgetMaterialTrack
       sections [{
         type=MovieSceneParameterSection
         end_frame=10
         start_frame=0
         channels [
          IsLeaping {type=MovieSceneScriptingFloatChannel, keys [[0,0],[6,1]]}
          RibbonRotation {type=MovieSceneScriptingFloatChannel, keys [[2,0],[4,0],[6,0.8],[10,0]]}
         ]
        }]
      }]
    }]}
 }
 Anim_EndAirDive {
  binding_count=1
  length_seconds=0.5
  object=&self:Anim_EndAirDive
  sequence {bindings [Image_Reticle {
     id=#4
     possessed_class=Image
     tracks [{
       type=MovieSceneWidgetMaterialTrack
       sections [{
         type=MovieSceneParameterSection
         end_frame=10
         start_frame=0
         channels [
          IsLeaping {type=MovieSceneScriptingFloatChannel, keys [[0,1],[3,1],[10,0]]}
          GlowOpacity {type=MovieSceneScriptingFloatChannel, keys [[0,2],[4,0]]}
          RibbonRotation {type=MovieSceneScriptingFloatChannel, keys [[0,0],[2,0.5],[6,-0.5]]}
         ]
        }]
      }]
    }]}
 }
 Anim_ChargeComplete {
  binding_count=1
  length_seconds=0.5
  object=&self:Anim_ChargeComplete
  sequence {bindings [Image_Reticle {
     id=#5
     possessed_class=Image
     tracks [{
       type=MovieSceneWidgetMaterialTrack
       sections [{
         type=MovieSceneParameterSection
         end_frame=10
         start_frame=0
         channels [
          IsLeaping {type=MovieSceneScriptingFloatChannel, const [1,1]}
          RibbonRotation {type=MovieSceneScriptingFloatChannel, keys [[0,0],[4,0.75]]}
          LeapChargeFill {type=MovieSceneScriptingFloatChannel, const [1,1]}
         ]
        }]
      }]
    }]}
 }
]
graphs [
 ConvertToSingleChargePercent {nodes [
   n1 = ConvertToSingleChargePercent!() -> (then, CurrentCharge:double=~) {is=title K2Node_FunctionEntry}
   n2 = GetValueAtLevel!(Input:"Scalable Float Structure (by ref)"=Get HF_MaxCharges() -> (HF_MaxCharges:Scalable Float Structure), Level:float=0) -> (ReturnValue:float)
   > Return Node!(execute, ReturnValue:double=float * float(A=float / float(A=1, B=@n2), B="Max (Float)"(A=@n1, B=0)) {is=title K2Node_PromotableOperator})
  ]}
 EventGraph {nodes [
   OnEquip() {is=title K2Node_CustomEvent}
   > ListenForStormClawsActiveTag()
   > ListenForAirLeapCooldownTag()
   > ListenForStormClawsAirDiveLoop()
   > PlayAnimationForward(InAnimation=Get Anim_Equip() -> (Anim_Equip:Widget Animation Object Reference))
   ListenForStormClawsActiveTag() {is=title K2Node_CustomEvent}
   > n1 = WaitGameplayTagAddToActor!(
    execute:Exec
    TargetActor:Actor Object Reference=Get PlayerController() -> (PlayerController:Player Controller Object Reference)
    Tag:Gameplay Tag Structure=Get StormClawsActiveTag() -> (StormClawsActiveTag:Gameplay Tag Structure)
    OnlyTriggerOnce:bool=false
   ) -> (then:Exec, AsyncTaskProxy:Ability Async Wait Gameplay Tag Added Object Reference, Added:Exec) {is=title K2Node_AsyncAction}
   > n2 = WaitGameplayTagRemoveFromActor!(
    execute:Exec
    TargetActor:Actor Object Reference=Get PlayerController()
    Tag:Gameplay Tag Structure=Get StormClawsActiveTag()
    OnlyTriggerOnce:bool=false
   ) -> (then:Exec, AsyncTaskProxy:Ability Async Wait Gameplay Tag Removed Object Reference, Removed:Exec) {is=title K2Node_AsyncAction}
   HandleStormClawsActive!(execute=@n2.Removed, self, SetActive:bool=false) -> (then)
   HandleStormClawsActive(execute=@n1.Added, SetActive=true)
   ListenForAirLeapCooldownTag() {is=title K2Node_CustomEvent}
   > n3 = WaitGameplayTagAddToActor(TargetActor=Get PlayerController(), Tag=Get AirLeapCooldownTag() -> (AirLeapCooldownTag:Gameplay Tag Structure), OnlyTriggerOnce=false)
   > n4 = WaitGameplayTagRemoveFromActor(TargetActor=Get PlayerController(), Tag=Get AirLeapCooldownTag(), OnlyTriggerOnce=false)
   n5 = Set Timer by Event(execute=@n6.then, Delegate=Create Event(), Time=0.1, bLooping=true, bMaxOncePerFrame=true)
   Clear and Invalidate Timer by Handle(execute=@n4.Removed, Handle=@n5)
   > HandleAirLeapCooldown!(execute, self, IsCooldownActive:bool=false) -> (then)
   n6 = HandleAirLeapCooldown(execute=@n3.Added, IsCooldownActive=true)
   HandleAirLeapCooldownReticleAnimation() {is=title K2Node_CustomEvent}
   > n7 = GetActiveEffectsWithAllTags!(
    execute
    self:Ability System Component Object Reference=GetAbilitySystemComponent!(Actor:Actor Object Reference=Get PlayerController()) -> (ReturnValue:Ability System Component Object Reference)
    Tags:Gameplay Tag Container Structure=Get AirLeapCooldownTagContainer() -> (AirLeapCooldownTagContainer:Gameplay Tag Container Structure)
   ) -> (then, ReturnValue:Array of Active Gameplay Effect Handle Structures)
   n8 = "Get (a ref)"(Array:Array of Active Gameplay Effect Handle Structures=@n7, Dimension 1=0) -> (Output:"Active Gameplay Effect Handle Structure (by ref)")
   > n9 = GetActiveGameplayEffectRemainingDuration!(execute, ActiveHandle:Active Gameplay Effect Handle Structure=@n8) -> (then, ReturnValue:float)
   > n10 = GetActiveGameplayEffectTotalDuration!(execute, ActiveHandle:Active Gameplay Effect Handle Structure=@n8) -> (then, ReturnValue:float)
   > n11 = GetDynamicMaterial(self=Get Image_Reticle() -> (Image_Reticle:Image Object Reference))
   > SetScalarParameterValue(self=@n11, ParameterName=LeapCooldown, Value=float - float(A=1, B=float / float(A=@n9, B=@n10)))
   ListenForStormClawsAirDiveLoop() {is=title K2Node_CustomEvent}
   > n12 = WaitGameplayTagAddToActor(TargetActor=Get PlayerController(), Tag=Get AirDiveTag() -> (AirDiveTag:Gameplay Tag Structure), OnlyTriggerOnce=false)
   > n13 = WaitGameplayTagRemoveFromActor(TargetActor=Get PlayerController(), Tag=Get AirDiveTag(), OnlyTriggerOnce=false)
   > n14 = WaitGameplayEventToActor!(
    execute:Exec
    TargetActor:Actor Object Reference=Get PlayerController()
    EventTag:Gameplay Tag Structure=(TagName="Ability.BeastGirl.StormClaws.StormSlam.StartedCharge")
    OnlyTriggerOnce:bool=false
    OnlyMatchExact:bool=true
   ) -> (then:Exec, AsyncTaskProxy:Ability Async Wait Gameplay Event Object Reference, EventReceived:Exec, Payload:Gameplay Event Data Structure) {is=title K2Node_AsyncAction}
   n15 = GetChargeAndDurationFromChargeStartedEvent!(EventData:"Gameplay Event Data Structure (by ref)"=@n14.Payload) -> (CurrentCharge:int=0, ChargeDuration:double)
   Set CurrentCharge(execute=@n14.EventReceived, CurrentCharge:int=@n15) -> (Output_Get:int=0)
   > Set ChargeDuration(ChargeDuration:float=@n15.ChargeDuration) -> (Output_Get:float)
   > Set StartedChargeTimestamp(StartedChargeTimestamp:float=GetServerWorldTimeSeconds!() -> (ReturnValue:float)) -> (Output_Get:float)
   > n16 = Set Timer by Event(Delegate=Create Event(), Time=0.05, bLooping=true, bMaxOncePerFrame=true)
   > Set AirLeapChargeReticleTimer(AirLeapChargeReticleTimer:Timer Handle Structure=@n16) -> (Output_Get:Timer Handle Structure)
   HandleAirDiveLoop!(execute=@n13.Removed, self, IsAirDiveLoopStart:bool=false) -> (then)
   > RemoveAirLeapTimers()
   Set CurrentCharge(execute=@n12.Added, CurrentCharge=0) -> (Output_Get=0)
   > HandleAirDiveLoop(IsAirDiveLoopStart=true)
   HandleAirLeapCharge() {is=title K2Node_CustomEvent}
   n17 = float - float(A=GetServerWorldTimeSeconds(), B=Get StartedChargeTimestamp() -> (StartedChargeTimestamp:float))
   > n18 = Branch(Condition="float >= float"(A=@n17, B=Get ChargeDuration() -> (ChargeDuration:float)))
   n19 = Get HF_MaxCharges()
   > Branch(Condition="float >= float"(A:int=Get CurrentCharge() -> (CurrentCharge:int=0), B=GetValueAtLevel(Input=@n19, Level=0)))
   > AbilityMaxCharge()
   n20 = float / float(A=@n17, B=Get ChargeDuration())
   n21 = Get CurrentCharge() -> (CurrentCharge=0)
   n22 = "To Float (Integer)"(InInt=@n21)
   SetScalarParameterValue(
    execute=@n18.else
    self=Get ImageMaterialInstanceRef() -> (ImageMaterialInstanceRef:Material Instance Dynamic Object Reference)
    ParameterName=LeapChargeFill
    Value=float + float(
     A=ConvertToSingleChargePercent!(CurrentCharge:double=@n20) -> (ReturnValue:double) {is=title K2Node_CallFunction}
     B=ConvertToSingleChargePercent(CurrentCharge=@n22)
    )
   )
   AbilityMaxCharge() {is=title K2Node_CustomEvent}
   > SetScalarParameterValue(self=Get ImageMaterialInstanceRef(), ParameterName=LeapChargeFill, Value=1)
   > PlayAnimationForward(InAnimation=Get Anim_ChargeComplete() -> (Anim_ChargeComplete:Widget Animation Object Reference))
   > RemoveAirLeapTimers()
   Event Construct()
   > Parent: Construct()
   > Set PlayerController(PlayerController:Player Controller Object Reference=GetOwningPlayer()) -> (Output_Get:Player Controller Object Reference)
   > OnEquip() {is=title K2Node_CallFunction}
   > n23 = GetDynamicMaterial(self=Get Image_Reticle())
   > Set ImageMaterialInstanceRef(ImageMaterialInstanceRef:Material Instance Dynamic Object Reference=@n23) -> (Output_Get:Material Instance Dynamic Object Reference)
   int - int(A=@n21)
  ]}
 HandleAirDiveLoop {nodes [
   HandleAirDiveLoop!() -> (then, IsAirDiveLoopStart:bool=~) {is=title K2Node_FunctionEntry}
   > SetScalarParameterValue(self=Get ImageMaterialInstanceRef(), ParameterName=LeapChargeFill, Value=0)
   > n1 = Branch(Condition=Get IsAirDiveLoopStart() -> (IsAirDiveLoopStart:bool))
   > n2 = Branch(Condition=Get IsCooldownActive() -> (IsCooldownActive:bool))
   PlayAnimationForward(execute=@n2.else, InAnimation=Get Anim_AirDiveLoop() -> (Anim_AirDiveLoop:Widget Animation Object Reference))
   PlayAnimationForward(execute=@n1.else, InAnimation=Get Anim_EndAirDive() -> (Anim_EndAirDive:Widget Animation Object Reference))
  ]}
 HandleAirLeapCooldown {nodes [
   n1 = HandleAirLeapCooldown!() -> (then, IsCooldownActive:bool=~) {is=title K2Node_FunctionEntry}
   > Set IsCooldownActive(IsCooldownActive:bool=@n1) -> (Output_Get:bool)
   > SetScalarParameterValue(
    self=Get ImageMaterialInstanceRef()
    ParameterName=LeapCooldown
    Value=SelectFloat(A=Get MinFillValue() -> (MinFillValue:double), B=Get MaxFillValue() -> (MaxFillValue:double), bPickA=Get IsCooldownActive())
   )
   > n2 = Branch(Condition=Get IsCooldownActive())
   > SetScalarParameterValue(self=Get ImageMaterialInstanceRef(), ParameterName=LeapChargeFill, Value=0)
   PlayAnimationForward(execute=@n2.else, InAnimation=Get Anim_CooldownComplete() -> (Anim_CooldownComplete:Widget Animation Object Reference))
  ]}
 HandleStormClawsActive {nodes [
   HandleStormClawsActive!() -> (then, SetActive:bool=~) {is=title K2Node_FunctionEntry}
   > SetVisibility(InVisibility=Select(Option 0=Collapsed, Option 1=HitTestInvisible, Index=Get SetActive() -> (SetActive:bool)))
  ]}
 RemoveAirLeapTimers {nodes [
   RemoveAirLeapTimers() {is=title K2Node_FunctionEntry}
   > Clear and Invalidate Timer by Handle(Handle=Get AirLeapChargeAbilityTimer() -> (AirLeapChargeAbilityTimer:Timer Handle Structure))
   > Clear and Invalidate Timer by Handle(Handle=Get AirLeapChargeReticleTimer() -> (AirLeapChargeReticleTimer:Timer Handle Structure))
  ]}
]
variables [
 AirDiveTag {type=GameplayTag, default {tagName=GameplayCue.Abilities.BeastGirl.StormClaws.AirDive.Loop}}
 AirDiveTagContainer {type=GameplayTagContainer, default {gameplayTags [{tagName=GameplayCue.Abilities.BeastGirl.StormClaws.AirDive.Loop}]}}
 AirLeapChargeAbilityTimer {type=TimerHandle}
 AirLeapChargeReticleTimer {type=TimerHandle}
 AirLeapCooldownTag {type=GameplayTag, default {tagName=Granted.BeastGirl.StormClaws.StormSlam.Cooldown}}
 AirLeapCooldownTagContainer {type=GameplayTagContainer, default {gameplayTags [{tagName=Granted.BeastGirl.StormClaws.StormSlam.Cooldown}]}}
 ChargeDuration {type=number}
 CurrentCharge {type=integer, default=-1}
 HF_ChargeExtraTime {
  type=ScalableFloat
  default {
   registryType=None
   value=1.333
   curve {curveTable=/BeastGirlStormClawsGameplay/DataTables/BeastGirlStormClawsGameplayGameData, rowName=BeastGirl.StormSlam.Charge.AdditionalAimTime}
  }
 }
 HF_MaxCharges {
  type=ScalableFloat
  default {
   registryType=None
   value=1
   curve {curveTable=&BeastGirlStormClawsGameplayGameData, rowName=BeastGirl.StormSlam.Charge.MaxCharges}
  }
 }
 ImageMaterialInstanceRef {type=MaterialInstanceDynamic}
 IsCooldownActive {type=boolean}
 LeapMaxChargeDuration {
  type=ScalableFloat
  default {
   registryType=None
   value=1
   curve {curveTable=&BeastGirlStormClawsGameplayGameData, rowName=BeastGirl.StormSlam.Charge.MaxDuration}
  }
 }
 PlayerController {type=PlayerController, default=null}
 StartedChargeTimestamp {type=number}
 StormClawsActiveTag {type=GameplayTag, default {tagName=Granted.BeastGirl.StormClaws.Active}}
]
widgets {
 info {
  inheritedWidgetCount=1
  parentClass=/Script/FortniteUI.FortWeaponReticleExtensionWidgetBase
  rootWidgetClass=&ScaleBox
  widgetCount=5
 }
 widgets [
  ScaleBox_0 {widgetClassPath=&ScaleBox, widget=ScaleBox_0, properties {stretch=UserSpecified}}
  Overlay_0 {widgetClassPath=&Overlay, parent=ScaleBox_0, widget=Overlay_0, slot {class=ScaleBoxSlot}}
  Image_Reticle {
   widgetClassPath=&Image
   bIsVariable=true
   parent=Overlay_0
   widget=Image_Reticle
   properties {
    brush {
     bIsDynamicallyLoaded=false
     drawAs=Image
     imageType=FullColor
     mirroring=NoMirror
     resourceName=None
     resourceObject=/BeastGirlStormClawsUI/Reticle/Materials/M_UI_BentBaton_Reticle
     tiling=NoTile
     imageSize=<161.2 104>
     margin=<ltrb 0.25 0 0.25 0>
     outlineSettings {
      bUseBrushTransparency=false
      roundingType=HalfHeightRadius
      width=0
      color {colorUseRule=UseColor_Specified, specifiedColor=<rgba 0 0 0 0>}
      cornerRadii=<0 0 0 0>
     }
     tintColor {colorUseRule=UseColor_Specified, specifiedColor=<rgba 1 1 1 1>}
     uVRegion {bIsValid=false, max=<0 0>, min=<0 0>}
    }
    renderTransform {angle=0, scale=<1 1>, shear=<0 0>, translation=<0 3.2>}
   }
   slot {class=OverlaySlot}
  }
  Image_Star {
   widgetClassPath=&Image
   bIsVariable=true
   parent=Overlay_0
   widget=Image_Star
   properties {brush {
     bIsDynamicallyLoaded=false
     drawAs=Image
     imageType=FullColor
     mirroring=NoMirror
     resourceName=None
     resourceObject=/BeastGirlStormClawsUI/HUD/Materials/M_UI_BentBaton_CircleCross
     tiling=NoTile
     imageSize=<10 10>
     margin=<ltrb 0 0 0 0>
     outlineSettings {
      bUseBrushTransparency=false
      roundingType=HalfHeightRadius
      width=0
      color {colorUseRule=UseColor_Specified, specifiedColor=<rgba 0 0 0 0>}
      cornerRadii=<0 0 0 0>
     }
     tintColor {colorUseRule=UseColor_Specified, specifiedColor=<rgba 1 1 1 1>}
     uVRegion {bIsValid=false, max=<0 0>, min=<0 0>}
    }}
   slot {class=OverlaySlot, properties {horizontalAlignment=HAlign_Center, verticalAlignment=VAlign_Center}}
  }
  TutorialHighlight {widgetClassPath=&TutorialHighlight}
 ]
}
